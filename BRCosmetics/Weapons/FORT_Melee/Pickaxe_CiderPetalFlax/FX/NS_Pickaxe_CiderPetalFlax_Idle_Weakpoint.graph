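# Niagara effect for the CiderPetalFlax pickaxe, named as its idle weakpoint FX, producing a short pink burst of mesh pieces and glow.
# Piece bursts normal-oriented meshes with a random visibility tag across two renderers, re-running a sphere location in update so a curve swells then shrinks their radius as they scale in and out.
# glow001 layers two growing sprite renderers, using secondary size and color attributes, with fading alpha scaled by time of day.
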NiagaraSystem /BRCosmetics/Weapons/FORT_Melee/Pickaxe_CiderPetalFlax/FX/NS_Pickaxe_CiderPetalFlax_Idle_Weakpoint
dependencies {
 usedDataInterfaces [&NiagaraDataInterfaceCurve]
 usedDynamicInputs [&ReturnNormalizedExecIndex]
 usedModules [&SystemState]
}
emitters {
 Piece {
  emitterUpdateScript [
   &EmitterState(Life Cycle Mode=::NewEnumerator1, Loop Behavior=::NewEnumerator1, Loop Duration=1, UseLoopDelay=false)
   &SpawnBurst_Instantaneous(Spawn Count=25, Use Spawn Probability=false, UseLoopCountLimit=false)
  ]
  particleSpawnScript [
   &InitializeParticle(
    Color Mode=::NewEnumerator0
    Lifetime Min=1.2
    Lifetime Mode=::NewEnumerator1
    Mesh Scale Max=<1.1 1 1.5>
    Mesh Scale Min=<0.6 0.6 0.6>
    Mesh Scale Mode=::NewEnumerator2
    UsePositionOffset=false
    Write Material Random=true
   )
   &ShapeLocation(
    MaskSpawnGroup=false
    Offset=<0 0 55>
    Offset Mode=::NewEnumerator0
    Sphere Distribution=::NewEnumerator3
    Sphere Radius=20
    UseOwnerScale=true
   ) {module=ShapeLocation001}
   &InitialMeshOrientation(Mesh Orientation Mode=::NewEnumerator2, Orientation Vector=$Particles.ShapeLocation001.ShapeNormal, Use Rotation Vector=false)
   /SetVariables(Particles.VisibilityTag=&RandomRangeInt(Minimum=0, Maximum=1)) {isSetParameters=true}
  ]
  particleUpdateScript [
   &ParticleState()
   &ScaleMeshSize(Scale Factor=&Multiply_VectorByFloat(
     Vector=<1 1 1>
     Float=&FloatFromCurve(FloatCurve=DataInterface() {propertyValues {Curve {keys [<at 0 0>,<at 0.1 1>,<at 0.6 1>,<at 0.65 0>,<at 1 0>]}}}, Scale Curve=1.2)
    ))
   &ShapeLocation(
    MaskSpawnGroup=false
    Offset=<0 0 55>
    Offset Mode=::NewEnumerator0
    Sphere Distribution=::NewEnumerator3
    Sphere Radius=&Multiply_Float(
     A=25
     B=&FloatFromCurve(FloatCurve=DataInterface() {propertyValues {Curve {
         defaultValue=1
         keys [
          {interpMode=RCIM_Cubic, tangentMode=RCTM_User, time=0, value=0}
          {interpMode=RCIM_Cubic, tangentMode=RCTM_User, time=0.2, value=2}
          {arriveTangent=-1.8632, interpMode=RCIM_Cubic, leaveTangent=-1.8632, tangentMode=RCTM_User, time=0.4, value=1.125}
          {tangentMode=RCTM_User, time=1, value=0.375}
         ]
        }}})
    )
    UseOwnerScale=true
   )
   /SetVariables(Particles.color002:LinearColor=&MakeLinearColorFromVectorAndFloat("Vector (RGB)"=&MakeVectorFromLinearColorRGB(Color=<rgba 0.1094 0.1094 0.1094 1>), "Float (Alpha)"=1)) {isSetParameters=true}
   &ScaleColor(
    Scale RGB=&Multiply_VectorByFloat(
     Vector=&MakeVectorFromLinearColorRGB(Color=<rgba 1 0.151 0.611 1>)
     Float=&RandomRangeFloat(Minimum=0.5, Maximum=1)
    )
    ScaleA=true
    ScaleRGB=true
   )
  ]
  rendererClasses [&NiagaraMeshRendererProperties]
  renderers [
   {rendererClass=&NiagaraMeshRendererProperties, rendererIndex=0}
   {rendererClass=&NiagaraMeshRendererProperties, rendererIndex=1}
  ]
 }
 glow001 {
  emitterUpdateScript [
   &EmitterState(Inactive Response=::NewEnumerator1, Life Cycle Mode=::NewEnumerator1, Loop Behavior=::NewEnumerator1, Loop Duration=1, UseLoopDelay=false)
   &SpawnBurst_Instantaneous(Use Spawn Probability=false, UseLoopCountLimit=false)
  ]
  particleSpawnScript [
   &InitializeParticle(
    Lifetime=1.2
    Position Offset=<0 0 55>
    Sprite Rotation Mode=::NewEnumerator2
    Sprite Size Mode=::NewEnumerator3
    Uniform Sprite Size=150
    UsePositionOffset=true
    Write Material Random=true
   )
   &CameraOffset(Camera Offset Amount=10)
  ]
  particleUpdateScript [
   &ParticleState()
   &ScaleSpriteSize(
    Uniform Curve Scale=2
    Uniform Curve Sprite Scale=DataInterface() {propertyValues {Curve {keys [<at 0 0.2>, <at 0.2 0.75>, {tangentMode=RCTM_User, time=1, value=1}]}}}
   )
   &ScaleColor(
    Scale Alpha=&FloatFromCurve(
     FloatCurve=DataInterface() {propertyValues {Curve {keys [
         {arriveTangent=-1.7607, interpMode=RCIM_Cubic, leaveTangent=-1.7607, tangentMode=RCTM_User, time=0, value=1}
         {arriveTangent=-0.1754, interpMode=RCIM_Cubic, leaveTangent=-0.1754, tangentMode=RCTM_User, time=0.4, value=0}
        ]}}}
     Scale Curve=&Multiply_Float(A=1, B=&FNTimeOfDayScalar(Night Scale=0.5, Morning Scale=0.8, Evening Scale=0.8))
    )
    Scale RGB=&MakeVectorFromLinearColorRGB(Color=<rgba 5 0.6843 2.6164 1>)
    ScaleA=true
    ScaleRGB=true
   )
   /SetVariables(Particles.Color002=&MakeLinearColorFromVectorAndFloat(
     "Vector (RGB)"=&MakeVectorFromLinearColorRGB(Color=<rgba 14.891 1.9018 15 1>)
     "Float (Alpha)"=&FloatFromCurve(
      FloatCurve=DataInterface() {propertyValues {Curve {keys [
          {arriveTangent=-1.7607, interpMode=RCIM_Cubic, leaveTangent=-1.7607, tangentMode=RCTM_User, time=0, value=1}
          {arriveTangent=-0.1754, interpMode=RCIM_Cubic, leaveTangent=-0.1754, tangentMode=RCTM_User, time=0.8, value=0}
         ]}}}
      Scale Curve=&Multiply_Float(A=1, B=&FNTimeOfDayScalar(Night Scale=1, Morning Scale=0.8, Evening Scale=0.8))
     )
    )) {isSetParameters=true}
   /SetVariables(Particles.Size002:Vector2f=&Multiply_Vector2D_ByFloat(Float=0.35, Vector2D=$Particles.SpriteSize)) {isSetParameters=true}
  ]
  rendererClasses [&NiagaraSpriteRendererProperties]
  renderers [
   {rendererClass=&NiagaraSpriteRendererProperties}
   {rendererClass=&NiagaraSpriteRendererProperties, rendererIndex=1}
  ]
 }
}
summary {
 systemName=NS_Pickaxe_CiderPetalFlax_Idle_Weakpoint
 emitters [
  {emitterName=glow001, rendererClasses [&NiagaraSpriteRendererProperties]}
  {emitterName=Piece, rendererClasses [&NiagaraMeshRendererProperties]}
 ]
 userVariables [
  User.bIsFrontEnd {type=NiagaraBool}
  User.bIsFrontEndPreview {type=NiagaraBool}
 ]
}
