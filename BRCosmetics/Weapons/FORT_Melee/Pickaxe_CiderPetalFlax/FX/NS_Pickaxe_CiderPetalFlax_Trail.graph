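# Niagara swing trail for the CiderPetalFlax pickaxe, combining pink ribbons, sparks and tumbling mesh pieces.
# Trail uses anim-trail modules to feed four ribbon renderers, with separate glow, light and highlight colors and widths.
# Sparks and piece spawn by distance moved at skeletal mesh locations, then drift under gravity, curl noise and drag.
# A front-end preview flag lowers trail brightness; time-of-day scalars adjust spark and glow alpha.
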NiagaraSystem /BRCosmetics/Weapons/FORT_Melee/Pickaxe_CiderPetalFlax/FX/NS_Pickaxe_CiderPetalFlax_Trail
dependencies {
 usedDataInterfaces [
  &NiagaraDataInterfaceColorCurve
  &NiagaraDataInterfaceCurve
  &NiagaraDataInterfaceSocketReader
  &NiagaraDataInterfaceSkeletalMesh
 ]
 usedDynamicInputs []
 usedModules [
  &self:SystemSpawnScript.SystemScriptSource.SystemScriptGraph.NiagaraNodeAssignment_0.SetVariables_#1
  &self:SystemSpawnScript.SystemScriptSource.SystemScriptGraph.NiagaraNodeAssignment_1.SetVariables_#2
  &SystemState
 ]
}
emitters {
 Sparks {
  emitterUpdateScript [
   &EmitterState(Life Cycle Mode=::NewEnumerator1, Loop Behavior=::NewEnumerator1, Loop Duration=0.5, UseLoopDelay=false)
   &SpawnPerUnit(Module.Use Max Movement Threshold=true, Module.Use Movement Tolerance=false, Use Spawn Probability=false)
  ]
  particleSpawnScript [
   &InitializeParticle(
    Color=<rgba 3 0.5 1.3841 1>
    Lifetime Max=0.5
    Lifetime Min=0.3
    Lifetime Mode=::NewEnumerator1
    Sprite Size Max=<5 10>
    Sprite Size Min=<5 15>
    Sprite Size Mode=::NewEnumerator2
    UsePositionOffset=false
    Write Material Random=true
   )
   &SkeletalMeshLocation(MaskSpawnGroup=false, Mesh Sampling Type=::NewEnumerator5, Skeletal Mesh=$System.Skeletal Mesh Reader, Socket Sampling Mode=::NewEnumerator5)
   &AddVelocity(Velocity=<0 0 100>)
   &BoxLocation(Box Size=<20 1 1>, MaskSpawnGroup=false, Module.Override Seed=false, Rotate Box=false, Surface Only=false)
   &ApplyInitialForces()
  ]
  particleUpdateScript [
   &ParticleState()
   /SetVariables(Particles.Glow03_Color=&MakeLinearColorFromVectorAndFloat("Vector (RGB)"=&MakeVectorFromLinearColorRGB(Color=<rgba 1 0.12 0.7067 1>), "Float (Alpha)"=0.1)) {isSetParameters=true}
   /SetVariables(Particles.Glow03_Size=&Multiply_Vector2D_ByFloat(
     Float=&FloatFromCurve(
      FloatCurve=DataInterface() {propertyValues {Curve {keys [
          <at 0 0 RCIM_Cubic>
          <at 0.2 1.3 RCIM_Cubic>
          {arriveTangent=-0.1423, interpMode=RCIM_Cubic, leaveTangent=-0.1423, tangentMode=RCTM_User, time=1, value=0}
         ]}}}
      Scale Curve=15
     )
     Vector2D=<1 2>
    )) {isSetParameters=true}
   &ScaleColor(
    Scale Alpha=&FloatFromCurve(
     FloatCurve=DataInterface() {propertyValues {Curve {keys [<at 0.3 1>,<at 0.4 1>,<at 1 1>]}}}
     Scale Curve=&Multiply_Float(A=1, B=&FNTimeOfDayScalar(Night Scale=0.6, Morning Scale=0.8, Evening Scale=0.8))
    )
    ScaleA=true
    ScaleRGB=true
   )
   &ScaleSpriteSize(Scale Factor=&Vector2DFromFloat(Value=&FloatFromCurve(
      FloatCurve=DataInterface() {propertyValues {Curve {keys [
          <at 0 0 RCIM_Cubic>
          {interpMode=RCIM_Cubic, leaveTangent=-3.0296, tangentMode=RCTM_Break, time=0.2, value=1.2}
          {arriveTangent=-1.8107, interpMode=RCIM_Cubic, leaveTangent=-1.8107, tangentMode=RCTM_User, time=1, value=0}
         ]}}}
      Scale Curve=0.4
     ))) {module=ScaleSpriteSize001}
   &GravityForce(Gravity=<0 0 -200>)
   &CurlNoiseForce(
    Noise Frequency=300
    Noise Strength=&Multiply_Float(A=1000, B=1)
    Pan Noise=true
    Pan Noise Field=<0 0 1>
    Random Seed=10
   )
   &Drag(Drag=3, Use Linear Drag=true, Use Rotational Drag=false)
   &SolveForcesAndVelocity(Clamp Velocity=false, Limit Acceleration=false)
  ]
  rendererClasses [&NiagaraSpriteRendererProperties]
  renderers [
   {rendererClass=&NiagaraSpriteRendererProperties}
   {rendererClass=&NiagaraSpriteRendererProperties, rendererIndex=1}
  ]
 }
 Trail {
  emitterUpdateScript [
   &EmitterState(Life Cycle Mode=::NewEnumerator1, Loop Behavior=::NewEnumerator1, Loop Duration=1, UseLoopDelay=false)
   &FN_SpawnAnimTrail()
  ]
  particleSpawnScript [
   &FN_InitializeAnimTrailBoneIndecies()
   &FN_InitializeAnimTrail(AnimTrailWidth=2, Lifetime=0.3)
  ]
  particleUpdateScript [
   &FN_UpdateAnimTrail()
   &ScaleColor(
    Scale Alpha=&FloatFromCurve(
     FloatCurve=DataInterface() {propertyValues {Curve {keys [
         {arriveTangent=-1.8769, interpMode=RCIM_Cubic, leaveTangent=-1.8769, tangentMode=RCTM_User, time=0, value=0.8}
         {arriveTangent=-1.0084, interpMode=RCIM_Cubic, leaveTangent=-1.0084, tangentMode=RCTM_User, time=0.3, value=0.25}
         <at 0.7 0 RCIM_Cubic>
        ]}}}
     Scale Curve=&Multiply_Float(
      A=&FloatFromCurve(FloatCurve=DataInterface() {propertyValues {Curve {keys [<at 0 1>,<at 0.4 1>,<at 0.6 0>]}}}, CurveIndex=$Emitter.Age)
      B=&MakeCustomFloatFromBool(Bool=$User.bIsFrontEndPreview, True Float=1.5, False Float=3)
     )
    )
    Scale RGB=&Multiply_VectorByFloat(
     Vector=&MakeVectorFromLinearColorRGB(Color=&ColorFromCurve(ColorCurve=DataInterface() {propertyValues {
         AlphaCurve {keys []}
         BlueCurve {keys [{arriveTangent=-0.034, interpMode=RCIM_Cubic, time=0.0778, value=0.4056}, <at 0.2725 0.8667>]}
         GreenCurve {keys [{arriveTangent=-0.0023, interpMode=RCIM_Cubic, time=0.0778, value=0.1458}, <at 0.2725 0.2>]}
         RedCurve {keys [{arriveTangent=-0.0931, interpMode=RCIM_Cubic, time=0.0778, value=1}, <at 0.2725 1>]}
        }}))
     Float=&MakeCustomFloatFromBool(Bool=$User.bIsFrontEndPreview, True Float=0.2, False Float=0.3)
    )
    ScaleA=true
    ScaleRGB=true
   )
   /SetVariables(Particles.Color_Glow:LinearColor=&MakeLinearColorFromVectorAndFloat(
     "Vector (RGB)"=&Multiply_VectorByFloat(
      Vector=&MakeVectorFromLinearColorRGB(Color=&ColorFromCurve(ColorCurve=DataInterface() {propertyValues {
          AlphaCurve {keys []}
          BlueCurve {keys [{arriveTangent=-0.034, interpMode=RCIM_Cubic, time=0.2143, value=2.2344}, <at 0.7864 0.0616>]}
          GreenCurve {keys [{arriveTangent=-0.0023, interpMode=RCIM_Cubic, time=0.2143, value=0.7031}, <at 0.7864 0.0182>]}
          RedCurve {keys [{arriveTangent=-0.0931, interpMode=RCIM_Cubic, time=0.2143, value=3}, <at 0.7864 0.0833>]}
         }}))
      Float=1
     )
     "Float (Alpha)"=&Multiply_Float(
      A=&MakeCustomFloatFromBool(Bool=$User.bIsFrontEndPreview, True Float=0.02, False Float=0.03)
      B=&FloatFromCurve(
       FloatCurve=DataInterface() {propertyValues {Curve {keys [<at 0 1>,<at 0.4 1>,<at 1 0>]}}}
       Scale Curve=&FNTimeOfDayScalar(Night Scale=0.5, Morning Scale=0.8, Evening Scale=0.8)
      )
     )
    )) {isSetParameters=true}
   /SetVariables(Particles.RibbonWidth_Glow:NiagaraFloat=&Multiply_Float(A=$Particles.RibbonWidth, B=1.5)) {isSetParameters=true}
   /SetVariables(Particles.Color_Light:LinearColor=&MakeLinearColorFromVectorAndFloat(
     "Vector (RGB)"=&Multiply_VectorByFloat(
      Vector=&MakeVectorFromLinearColorRGB(Color=&ColorFromCurve(ColorCurve=DataInterface() {propertyValues {
          AlphaCurve {keys []}
          BlueCurve {keys [
            {arriveTangent=-11.7847, interpMode=RCIM_Cubic, leaveTangent=-11.7847, tangentMode=RCTM_User, time=0.0425, value=5}
            <at 0.1942 2.1>
            <at 0.5887 0.2822>
           ]}
          GreenCurve {keys [
            {arriveTangent=-4.0036, interpMode=RCIM_Cubic, leaveTangent=-4.0036, tangentMode=RCTM_User, time=0.0425, value=0.8}
            <at 0.1942 0.3>
            <at 0.5887 0.0467>
           ]}
          RedCurve {keys [
            {arriveTangent=-34.1305, interpMode=RCIM_Cubic, leaveTangent=-34.1305, tangentMode=RCTM_User, time=0.0425, value=8}
            <at 0.1942 3>
            <at 0.5887 0.4>
           ]}
         }}))
      Float=1
     )
     "Float (Alpha)"=&Multiply_Float(
      A=&MakeCustomFloatFromBool(Bool=$User.bIsFrontEndPreview, True Float=0.35, False Float=0.5)
      B=&FloatFromCurve(FloatCurve=DataInterface() {propertyValues {Curve {keys [
           <at 0 1>
           {arriveTangent=-1.716, interpMode=RCIM_Cubic, leaveTangent=-1.716, tangentMode=RCTM_User, time=0.6, value=1}
           {arriveTangent=0.05, interpMode=RCIM_Cubic, leaveTangent=0.05, tangentMode=RCTM_User, time=0.9, value=0}
          ]}}})
     )
    )) {isSetParameters=true}
   /SetVariables(Particles.RibbonWidth_light:NiagaraFloat=&Multiply_Float(A=$Particles.RibbonWidth, B=1.4)) {isSetParameters=true}
   /SetVariables(Particles.Color_highlight:LinearColor=&MakeLinearColorFromVectorAndFloat(
     "Vector (RGB)"=&Multiply_VectorByFloat(
      Vector=&MakeVectorFromLinearColorRGB(Color=&ColorFromCurve(ColorCurve=DataInterface() {propertyValues {
          AlphaCurve {keys []}
          BlueCurve {keys [{arriveTangent=-0.034, interpMode=RCIM_Cubic, time=0.1168, value=13.6667}, <at 0.3582 0.324>]}
          GreenCurve {keys [{arriveTangent=-0.0023, interpMode=RCIM_Cubic, time=0.1168, value=1}, <at 0.3582 0.0344>]}
          RedCurve {keys [{arriveTangent=-0.0931, interpMode=RCIM_Cubic, time=0.1168, value=20}, <at 0.3582 0.4688>]}
         }}))
      Float=1
     )
     "Float (Alpha)"=&Multiply_Float(
      A=1
      B=&FloatFromCurve(FloatCurve=DataInterface() {propertyValues {Curve {keys [<at 0 1>,<at 0.3 1>,<at 0.6 0>]}}})
     )
    )) {isSetParameters=true}
   /SetVariables(Particles.RibbonWidth_highlight:NiagaraFloat=&Multiply_Float(A=$Particles.RibbonWidth, B=1)) {isSetParameters=true}
  ]
  rendererClasses [&NiagaraRibbonRendererProperties]
  renderers [
   {rendererClass=&NiagaraRibbonRendererProperties}
   {rendererClass=&NiagaraRibbonRendererProperties, rendererIndex=1}
   {rendererClass=&NiagaraRibbonRendererProperties, rendererIndex=2}
   {rendererClass=&NiagaraRibbonRendererProperties, rendererIndex=3}
  ]
 }
 piece {
  emitterUpdateScript [
   &EmitterState(Life Cycle Mode=::NewEnumerator1, Loop Behavior=::NewEnumerator1, Loop Delay=0.04, Loop Duration=0.5, UseLoopDelay=true)
   &SpawnPerUnit(Module.Use Max Movement Threshold=true, Module.Use Movement Tolerance=false, Spawn Spacing=30, Use Spawn Probability=false)
  ]
  particleSpawnScript [
   &InitializeParticle(
    Color Mode=::NewEnumerator0
    Lifetime Max=0.4
    Lifetime Min=0.3
    Lifetime Mode=::NewEnumerator1
    Mesh Scale Max=<1.1 1.1 1.1>
    Mesh Scale Min=<0.6 0.6 0.6>
    Mesh Scale Mode=::NewEnumerator2
    UsePositionOffset=false
    Write Material Random=true
   )
   &SkeletalMeshLocation(
    MaskSpawnGroup=false
    Mesh Sampling Type=::NewEnumerator5
    Sampled Position Offset=<0 0 20>
    Skeletal Mesh=$System.Skeletal Mesh Reader
    Socket Sampling Mode=::NewEnumerator5
   )
   &AddVelocity()
   &BoxLocation(Box Size=<20 20 20>, MaskSpawnGroup=false, Module.Override Seed=false, Rotate Box=false, Surface Only=false)
   &InitialMeshOrientation(
    Mesh Orientation Mode=::NewEnumerator2
    Rotation=&RandomRangeVector(Minimum=<-1 -1 -1>, Maximum=<1 1 1>)
    Use Rotation Vector=true
   )
   /SetVariables(Particles.VisibilityTag=&RandomRangeInt(Minimum=0, Maximum=1)) {isSetParameters=true}
  ]
  particleUpdateScript [
   &ParticleState()
   &ScaleMeshSize(Scale Factor=&Multiply_VectorByFloat(
     Vector=<1 1 1>
     Float=&FloatFromCurve(FloatCurve=DataInterface() {propertyValues {Curve {keys [<at 0 1>,<at 0.8 1>,<at 1 0>]}}})
    ))
   &UpdateMeshOrientation(
    Rotation Rate=&FloatFromCurve(
     FloatCurve=DataInterface() {propertyValues {Curve {keys [<at 0 1 RCIM_Cubic>, {arriveTangent=-1.7112, interpMode=RCIM_Cubic, leaveTangent=-1.7112, tangentMode=RCTM_User, time=1, value=0}]}}}
     Scale Curve=2
    )
    Rotation Vector=<1 0.2 0.3>
   )
   &GravityForce(Gravity=&RandomRangeVector(Minimum=<0 0 -260>, Maximum=<0 0 -350>))
   &CurlNoiseForce(
    Noise Frequency=300
    Noise Strength=&Multiply_Float(A=500, B=1)
    Pan Noise=true
    Pan Noise Field=<1 0 0>
    Random Seed=10
   )
   &Drag(Use Linear Drag=true, Use Rotational Drag=true)
   &SolveForcesAndVelocity(Clamp Velocity=false, Limit Acceleration=false)
   &ScaleColor(
    Scale RGB=&Multiply_VectorByFloat(Vector=&MakeVectorFromLinearColorRGB(Color=<rgba 1 0.1354 0.6038 1>), Float=1)
    ScaleA=true
    ScaleRGB=true
   )
   /SetVariables(Particles.color002:LinearColor=&MakeLinearColorFromVectorAndFloat("Vector (RGB)"=&MakeVectorFromLinearColorRGB(Color=<rgba 0.2396 0.2396 0.2396 1>), "Float (Alpha)"=1)) {isSetParameters=true}
  ]
  rendererClasses [&NiagaraMeshRendererProperties]
  renderers [
   {rendererClass=&NiagaraMeshRendererProperties, rendererIndex=0}
   {rendererClass=&NiagaraMeshRendererProperties, rendererIndex=1}
  ]
 }
}
summary {
 systemName=NS_Pickaxe_CiderPetalFlax_Trail
 emitters [
  {emitterName=Trail, rendererClasses [&NiagaraRibbonRendererProperties]}
  {emitterName=Sparks, rendererClasses [&NiagaraSpriteRendererProperties]}
  {emitterName=piece, rendererClasses [&NiagaraMeshRendererProperties]}
 ]
 userVariables [
  User.bIsFrontEndPreview {type=NiagaraBool}
  User.RibbonIDOffset {type=NiagaraInt32}
 ]
}
