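# Material function for the CiderPetalFlax pickaxe that blends a faux-depth crystal look into incoming material attributes by an Alpha input, overriding base color and emissive.
# Bump-offset crystal textures fake interior depth, and two multiplied panning samples form squared sparkles.
# It also builds sky-atmosphere-light backlight, squared fresnel and skylight environment reflection terms.
# A Reactivity term takes the maximum of timestamp-driven pulses for elimination, damage and weak-point events, each with loop preview.
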MaterialFunction /BRCosmetics/Weapons/FORT_Melee/Pickaxe_CiderPetalFlax/Materials/MF_CiderPetalFlax_Crystal
graph {nodes [
  Constant()
  FunctionInput() {bUsePreviewValueAsDefault=true, inputName=--- COLOR ---}
  FunctionInput() {bUsePreviewValueAsDefault=true, inputName=--- INTERIOR ---, sortPriority=10}
  FunctionInput() {bUsePreviewValueAsDefault=true, inputName=--- SPARKS ---, sortPriority=20}
  FunctionInput() {bUsePreviewValueAsDefault=true, inputName=--- BACKLIGHT ---, sortPriority=30}
  FunctionInput() {bUsePreviewValueAsDefault=true, inputName=--- FRESNEL ---, sortPriority=40}
  FunctionInput() {bUsePreviewValueAsDefault=true, inputName=--- REFLECTION ---, sortPriority=50}
  n1 = NamedRerouteUsage()
  n2 = NamedRerouteUsage()
  n3 = LinearInterpolate(A=Multiply(A=@n2, B=NamedRerouteUsage()), B=@n2, Alpha=NamedRerouteUsage())
  n4 = TextureSample(UVs=NamedRerouteUsage()) {texture=/Game/Global/Textures/Default/Masks_and_Gradients/T_MF_Crystal}
  n5 = OneMinus(None=NamedRerouteUsage())
  n6 = Multiply(A=Multiply(A=@n3, B=Desaturation(None=@n4.RGB, Fraction=@n5)), B=NamedRerouteUsage()) {constB=0.1}
  n7 = NamedRerouteUsage()
  n8 = LinearInterpolate(
   A=NamedRerouteDeclaration(None=LinearInterpolate(A=@n3, B=@n6, Alpha=@n7)) {name=InteriorColor, nodeColor=<rgba 1 0 0.4687 1>}
   B=NamedRerouteUsage()
   Alpha=NamedRerouteUsage()
  )
  n9 = NamedRerouteUsage()
  n10 = NamedRerouteUsage()
  n11 = Add(A=Add(A=Add(A=@n9, B=@n10), B=NamedRerouteUsage()), B=NamedRerouteUsage())
  FunctionOutput(None=BlendMaterialAttributes(
    A=NamedRerouteUsage()
    B=SetMaterialAttributes!(MaterialAttributes=@n1, BaseColor=@n8, EmissiveColor=@n11) {attributeSetTypes [#1, #2]}
    Alpha=FunctionInput() {inputName=Alpha, inputType=FunctionInput_Scalar, sortPriority=-5, previewValue=<1 1 1 1>}
   ))
  n12 = GetMaterialAttributes(None=FunctionInput() {inputName=Material, inputType=FunctionInput_MaterialAttributes, sortPriority=-6}) {attributeGetTypes [#1, #2]}
  NamedRerouteDeclaration(None=@n12.EmissiveColor) {name=Emissive, nodeColor=<rgba 1 0 0.75 1>}
  NamedRerouteDeclaration(None=@n12.MaterialAttributes) {name=MA, nodeColor=<rgba 0.3406 1 0.0156 1>}
  n13 = TextureSample(UVs=NamedRerouteUsage()) {
   samplerSource=SSM_Wrap_WorldGroupSettings
   samplerType=SAMPLERTYPE_Normal
   texture=/Game/Global/Textures/Default/Masks_and_Gradients/T_MF_Crystal3
  }
  n14 = Transform(None=LinearInterpolate(A=Constant3Vector() {constant=<rgba 0 0 1 0>}, B=@n13.RGB, Alpha=NamedRerouteUsage()) {constAlpha=0.1})
  n15 = CameraVectorWS()
  n16 = Multiply(A=NamedRerouteUsage(), B=@n15)
  n17 = DotProduct(A=Normalize(VectorInput=Add(A=@n14, B=@n16)), B=Multiply(A=SkyAtmosphereLightDirection()) {constB=-1})
  n18 = Saturate(None=Add(A=Multiply(A=@n17) {constB=0.5}) {constB=0.5})
  n19 = NamedRerouteUsage()
  n20 = DotProduct(A=@n14, B=@n15)
  n21 = Multiply(A=Multiply(A=Power(Base=@n18, Exp=@n19), B=Saturate(None=@n20)), B=NamedRerouteUsage())
  n22 = Multiply(A=NamedRerouteUsage(), B=NamedRerouteUsage())
  NamedRerouteDeclaration(None=Multiply(A=Multiply(A=@n21, B=@n22), B=NamedRerouteUsage())) {name=Backlight, nodeColor=<rgba 1 0 0.2578 1>}
  NamedRerouteDeclaration(None=FunctionInput() {bUsePreviewValueAsDefault=true, inputName=Base Color, sortPriority=1, previewValue=<1 1 1 1>}) {name=IN_BaseColor, nodeColor=<rgba 1 0.2344 0 1>}
  NamedRerouteDeclaration(None=FunctionInput() {bUsePreviewValueAsDefault=true, inputName=Base Color Multiply/Override, inputType=FunctionInput_Scalar, sortPriority=2}) {name=IN_BaseColor_Mult/Override, nodeColor=<rgba 0 0.0859 1 1>}
  n23 = FunctionInput() {bUsePreviewValueAsDefault=true, inputName=Interior Texture Tile, inputType=FunctionInput_Scalar, sortPriority=11, previewValue=<1 0 0 1>}
  NamedRerouteDeclaration(None=@n23) {name=IN_InteriorUVTiling, nodeColor=<rgba 0.4062 1 0 1>}
  NamedRerouteDeclaration(None=FunctionInput() {bUsePreviewValueAsDefault=true, inputName=Interior Depth Minimum, inputType=FunctionInput_Scalar, sortPriority=12}) {name=IN_InteriorDepthMin, nodeColor=<rgba 1 0 0.7266 1>}
  n24 = FunctionInput() {bUsePreviewValueAsDefault=true, inputName=Interior Depth Maximum, inputType=FunctionInput_Scalar, sortPriority=13, previewValue=<-10 0 0 1>}
  NamedRerouteDeclaration(None=@n24) {name=IN_InteriorDepthMax, nodeColor=<rgba 0 1 0.9531 1>}
  n25 = FunctionInput() {
   bUsePreviewValueAsDefault=true
   inputName=Interior Depth Texture Tile
   inputType=FunctionInput_Scalar
   sortPriority=14
   previewValue=<2.6 0 0 1>
  }
  NamedRerouteDeclaration(None=@n25) {name=IN_InteriorDepthTextureTile, nodeColor=<rgba 1 0.6328 0 1>}
  n26 = FunctionInput() {bUsePreviewValueAsDefault=true, inputName=Sparks Tile, inputType=FunctionInput_Scalar, sortPriority=22, previewValue=<2 0 0 1>}
  NamedRerouteDeclaration(None=@n26) {name=IN_SparksTile, nodeColor=<rgba 0.3125 0 1 1>}
  n27 = FunctionInput() {bUsePreviewValueAsDefault=true, inputName=Sparks Depth, inputType=FunctionInput_Scalar, sortPriority=23, previewValue=<-5 0 0 1>}
  NamedRerouteDeclaration(None=@n27) {name=IN_SparksDepth, nodeColor=<rgba 0.0078 1 0 1>}
  n28 = FunctionInput() {bUsePreviewValueAsDefault=true, inputName=SparksSpeed_01, inputType=FunctionInput_Scalar, sortPriority=24, previewValue=<0.01 0 0 1>}
  NamedRerouteDeclaration(None=@n28) {name=IN_SparksSpeed_01, nodeColor=<rgba 1 0 0.3281 1>}
  NamedRerouteDeclaration(None=@n12.BaseColor) {name=BaseColor, nodeColor=<rgba 0.1661 0.3253 1 1>}
  n29 = FunctionInput() {bUsePreviewValueAsDefault=true, inputName=SparksSpeed_02, inputType=FunctionInput_Scalar, sortPriority=25, previewValue=<-0.005 0 0 1>}
  NamedRerouteDeclaration(None=@n29) {name=IN_SparksSpeed_02, nodeColor=<rgba 1 0 0.3281 1>}
  n30 = FunctionInput() {bUsePreviewValueAsDefault=true, inputName=Sparks Intensity, inputType=FunctionInput_Scalar, sortPriority=21, previewValue=<1 0 0 1>}
  NamedRerouteDeclaration(None=@n30) {name=IN_SparksIntensity, nodeColor=<rgba 0.9688 1 0 1>}
  n31 = FunctionInput() {bUsePreviewValueAsDefault=true, inputName=Interior Prism Colors, inputType=FunctionInput_Scalar, sortPriority=15, previewValue=<0.5 0 0 1>}
  NamedRerouteDeclaration(None=@n31) {name=IN_InteriorPrismColors, nodeColor=<rgba 0.4297 1 0 1>}
  n32 = FunctionInput() {bUsePreviewValueAsDefault=true, inputName=Interior Intensity, inputType=FunctionInput_Scalar, sortPriority=16, previewValue=<1 0 0 1>}
  NamedRerouteDeclaration(None=@n32) {name=IN_InteriorIntensity, nodeColor=<rgba 1 0 0.75 1>}
  n33 = FunctionInput() {bUsePreviewValueAsDefault=true, inputName=Interior Contrast, inputType=FunctionInput_Scalar, sortPriority=17, previewValue=<0.5 0 0 1>}
  NamedRerouteDeclaration(None=@n33) {name=IN_InteriorContrast, nodeColor=<rgba 0 1 0.9297 1>}
  n34 = FunctionInput() {bUsePreviewValueAsDefault=true, inputName=Backlight Color, sortPriority=31, previewValue=<1 1 1 1>}
  NamedRerouteDeclaration(None=@n34) {name=IN_BacklightColor, nodeColor=<rgba 0 1 0.6016 1>}
  n35 = FunctionInput() {bUsePreviewValueAsDefault=true, inputName=Backlight Intensity, inputType=FunctionInput_Scalar, sortPriority=32, previewValue=<1 0 0 1>}
  NamedRerouteDeclaration(None=@n35) {name=IN_BacklightIntensity, nodeColor=<rgba 1 0.2812 0 1>}
  n36 = FunctionInput() {
   bUsePreviewValueAsDefault=true
   inputName=Backlight Normal Intensity
   inputType=FunctionInput_Scalar
   sortPriority=33
   previewValue=<0.2 0 0 1>
  }
  NamedRerouteDeclaration(None=@n36) {name=IN_BacklightNormalIntensity, nodeColor=<rgba 1 0.2812 0 1>}
  n37 = FunctionInput() {bUsePreviewValueAsDefault=true, inputName=Backlight View Angle, inputType=FunctionInput_Scalar, sortPriority=34, previewValue=<1 0 0 1>}
  NamedRerouteDeclaration(None=@n37) {name=IN_BacklightViewAngle, nodeColor=<rgba 0.3594 1 0 1>}
  n38 = FunctionInput() {bUsePreviewValueAsDefault=true, inputName=Backlight Aperture, inputType=FunctionInput_Scalar, sortPriority=35, previewValue=<2 0 0 1>}
  NamedRerouteDeclaration(None=@n38) {name=IN_BacklightAperture, nodeColor=<rgba 1 0 0.6797 1>}
  n39 = ReflectionVectorWS()
  n40 = NamedRerouteUsage()
  NamedRerouteDeclaration(None=Multiply(A=SkyLightEnvMapSample(Direction=@n39, Roughness=@n40), B=NamedRerouteUsage())) {name=Reflection, nodeColor=<rgba 1 0.1406 0 1>}
  n41 = FunctionInput() {bUsePreviewValueAsDefault=true, inputName=Fresnel Color, sortPriority=41, previewValue=<1 1 1 1>}
  NamedRerouteDeclaration(None=@n41) {name=IN_FresnelColor, nodeColor=<rgba 0 1 1 1>}
  n42 = FunctionInput() {bUsePreviewValueAsDefault=true, inputName=Fresnel Intensity, inputType=FunctionInput_Scalar, sortPriority=42, previewValue=<0.2 0 0 1>}
  NamedRerouteDeclaration(None=@n42) {name=IN_FresnelIntensity, nodeColor=<rgba 1 0.6797 0 1>}
  n43 = FunctionInput() {bUsePreviewValueAsDefault=true, inputName=Reflection Color, sortPriority=51, previewValue=<1 1 1 1>}
  NamedRerouteDeclaration(None=@n43) {name=IN_ReflectionColor, nodeColor=<rgba 0.3594 0 1 1>}
  n44 = FunctionInput() {bUsePreviewValueAsDefault=true, inputName=Reflection Roughness, inputType=FunctionInput_Scalar, sortPriority=52, previewValue=<0.2 0 0 1>}
  NamedRerouteDeclaration(None=@n44) {name=IN_ReflectionRoughness, nodeColor=<rgba 0 1 0.0391 1>}
  n45 = Divide(
   A=Subtract(A=Time(), B=ScalarParameter() {defaultValue=-10000, group=Reactivity, parameterName=Timestamp_Elim})
   B=ScalarParameter() {defaultValue=5, group=Reactivity, parameterName=ElimReactivity_Duration}
  )
  n46 = LinearInterpolate(A=Saturate(None=@n45), B=Frac(None=@n45), Alpha=ScalarParameter() {group=Reactivity, parameterName=ElimReactivity_LoopPreview})
  n47 = Abs(None=ConstantBiasScale(None=@n46) {bias=-0.5, scale=2})
  n48 = ScalarParameter() {defaultValue=0.5, group=Reactivity, parameterName=ElimReactivity_StartExp}
  n49 = ScalarParameter() {defaultValue=2, group=Reactivity, parameterName=ElimReactivity_EndExp}
  n50 = Round(None=@n46)
  n51 = LinearInterpolate(
   B=ScalarParameter() {defaultValue=100, group=Reactivity, parameterName=ElimReactivity_Intensity}
   Alpha=Power(Base=OneMinus(None=@n47), Exp=LinearInterpolate(A=@n48, B=@n49, Alpha=@n50))
  ) {constA=1}
  n52 = Divide(
   A=Subtract(A=Time(), B=ScalarParameter() {defaultValue=-10000, group=Reactivity, parameterName=Timestamp_Dmg})
   B=ScalarParameter() {defaultValue=1.5, group=Reactivity, parameterName=DmgReactivity_Duration}
  )
  n53 = LinearInterpolate(A=Saturate(None=@n52), B=Frac(None=@n52), Alpha=ScalarParameter() {group=Reactivity, parameterName=DmgReactivity_LoopPreview})
  n54 = Abs(None=ConstantBiasScale(None=@n53) {bias=-0.5, scale=2})
  n55 = ScalarParameter() {defaultValue=0.5, group=Reactivity, parameterName=DmgReactivity_StartExp}
  n56 = ScalarParameter() {defaultValue=2, group=Reactivity, parameterName=DmgReactivity_EndExp}
  n57 = Round(None=@n53)
  n58 = LinearInterpolate(
   B=ScalarParameter() {defaultValue=80, group=Reactivity, parameterName=DmgReactivity_Intensity}
   Alpha=Power(Base=OneMinus(None=@n54), Exp=LinearInterpolate(A=@n55, B=@n56, Alpha=@n57))
  ) {constA=1}
  n59 = ScalarParameter() {defaultValue=150, group=Reactivity, parameterName=WeakPointgReactivity_Intensity}
  n60 = Divide(
   A=Subtract(A=Time(), B=ScalarParameter() {defaultValue=-10000, group=Reactivity, parameterName=Timestamp_WeakPoint})
   B=ScalarParameter() {defaultValue=1.5, group=Reactivity, parameterName=WeakPointReactivity_Duration}
  )
  n61 = LinearInterpolate(A=Saturate(None=@n60), B=Frac(None=@n60), Alpha=ScalarParameter() {group=Reactivity, parameterName=WeakPointReactivity_LoopPreview})
  n62 = ConstantBiasScale(None=@n61) {bias=-0.5, scale=2}
  n63 = Power(
   Base=OneMinus(None=Abs(None=@n62))
   Exp=LinearInterpolate(
    A=ScalarParameter() {defaultValue=0.5, group=Reactivity, parameterName=WeakPointReactivity_StartExp}
    B=ScalarParameter() {defaultValue=2, group=Reactivity, parameterName=WeakPointReactivity_EndExp}
    Alpha=Round(None=@n61)
   )
  )
  NamedRerouteDeclaration(None=Max(A=Max(A=@n51, B=@n58), B=LinearInterpolate(B=@n59, Alpha=@n63) {constA=1})) {name=Reactivity, nodeColor=<rgba 1 0.1875 0 1>}
  n64 = TextureCoordinate()
  n65 = NamedRerouteDeclaration(None=@n64) {name=UVCoord, nodeColor=<rgba 0.5937 1 0 1>}
  n66 = NamedRerouteUsage()
  n67 = NamedRerouteUsage()
  n68 = NamedRerouteUsage()
  n69 = TextureSample(UVs=Multiply(A=@n64, B=NamedRerouteUsage())) {
   samplerSource=SSM_Wrap_WorldGroupSettings
   samplerType=SAMPLERTYPE_LinearGrayscale
   texture=/Game/Global/Textures/Default/Masks_and_Gradients/T_MF_Crystal2
  }
  n70 = NamedRerouteDeclaration(None=@n69.R) {name=SparksDepth, nodeColor=<rgba 0.1156 1 0.3086 1>}
  NamedRerouteDeclaration(None=BumpOffset(Coordinate=Multiply(A=@n65, B=@n66), Height=LinearInterpolate(A=@n67, B=@n68, Alpha=@n70) {constA=-4, constB=-20})) {name=UVOffsetCoord, nodeColor=<rgba 0.9453 0 1 1>}
  n71 = PixelNormalWS()
  n72 = CameraVectorWS()
  n73 = OneMinus(None=Saturate(None=DotProduct(A=@n71, B=@n72)))
  NamedRerouteDeclaration(None=Multiply(A=Multiply(A=@n73, B=@n73), B=NamedRerouteUsage())) {name=Fresnel, nodeColor=<rgba 0 0.4609 1 1>}
  n74 = BumpOffset(Coordinate=NamedRerouteUsage(), Height=Multiply(A=NamedRerouteUsage(), B=NamedRerouteUsage())) {heightRatio=0.01}
  n75 = NamedRerouteUsage()
  n76 = Multiply(A=@n75) {constB=1.2}
  n77 = Time()
  n78 = NamedRerouteUsage()
  n79 = TextureObject() {samplerType=SAMPLERTYPE_LinearGrayscale, texture=/Game/Global/Textures/Default/Masks_and_Gradients/T_MF_CrystalSparks}
  n80 = TextureSample(UVs=Add(A=Multiply(A=@n74, B=@n76), B=Multiply(A=@n77, B=@n78)), Tex=@n79) {
   samplerSource=SSM_Wrap_WorldGroupSettings
   samplerType=SAMPLERTYPE_LinearGrayscale
   texture=/Game/Accessories/FORT_Backpacks/Backpack_F_MED_RoyalDusk/Textures/T_RoyalDusk_RoyalDusk_CrystalSparks1
  }
  n81 = NamedRerouteUsage()
  n82 = NamedRerouteUsage()
  n83 = TextureSample(UVs=Add(A=Multiply(A=@n75, B=@n81), B=Multiply(A=@n77, B=@n82)), Tex=@n79) {
   samplerSource=SSM_Wrap_WorldGroupSettings
   samplerType=SAMPLERTYPE_LinearGrayscale
   texture=&T_RoyalDusk_RoyalDusk_CrystalSparks1
  }
  n84 = NamedRerouteUsage()
  n85 = NamedRerouteUsage()
  n86 = Multiply(A=Multiply(A=Multiply(A=@n80.R, B=@n83.R), B=Add(A=@n84, B=@n85)) {constB=0}, B=NamedRerouteUsage()) {constB=16}
  NamedRerouteDeclaration(None=Multiply(A=@n86, B=@n86)) {name=Sparks, nodeColor=<rgba 1 0.3611 0.0777 1>}
 ]}
