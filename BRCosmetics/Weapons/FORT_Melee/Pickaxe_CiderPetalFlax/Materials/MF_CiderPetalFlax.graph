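# Material function for the CiderPetalFlax pickaxe that layers pink and blue crystal looks onto an incoming material.
# Channel-selectable Tech Art Mask channels define each crystal, with per-crystal color, interior and reflection parameters lerped between the two.
# A tiled detail normal, with per-crystal strength, is angle-corrected-blended into the base normal.
# MF_CiderPetalFlax_Crystal builds the final result with depth interior, sparks, backlight, fresnel and reflection inputs.
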MaterialFunction /BRCosmetics/Weapons/FORT_Melee/Pickaxe_CiderPetalFlax/Materials/MF_CiderPetalFlax
graph {nodes [
  n1 = NamedRerouteUsage()
  n2 = Constant() {r=-20}
  n3 = NamedRerouteUsage()
  n4 = VectorParameter() {defaultValue=<rgba 0.25 0 0.125 1>}
  n5 = Constant() {r=0.3}
  n6 = NamedRerouteUsage()
  FunctionOutput(None=MaterialFunctionCall!(
    Material=NamedRerouteUsage()
    Alpha=NamedRerouteUsage()
    --- COLOR ---
    Base Color=NamedRerouteUsage()
    Base Color Multiply/Override=@n1
    --- INTERIOR ---
    Interior Texture Tile=Constant() {r=3}
    Interior Depth Minimum=Multiply(A=@n1, B=@n2)
    Interior Depth Maximum
    Interior Depth Texture Tile=Constant() {r=10}
    Interior Prism Colors=NamedRerouteUsage()
    Interior Intensity=NamedRerouteUsage()
    Interior Contrast=NamedRerouteUsage()
    --- SPARKS ---
    Sparks Intensity=Multiply(A=@n3) {constB=5}
    Sparks Tile=Constant() {r=3}
    Sparks Depth=Constant() {r=-7}
    SparksSpeed_01
    SparksSpeed_02
    --- BACKLIGHT ---
    Backlight Color=Multiply(A=@n4.RGB, B=@n3)
    Backlight Intensity
    Backlight Normal Intensity=ScalarParameter() {defaultValue=1}
    Backlight View Angle
    Backlight Aperture
    --- FRESNEL ---
    Fresnel Color
    Fresnel Intensity
    --- REFLECTION ---
    Reflection Color=NamedRerouteUsage()
    Reflection Roughness=Multiply(A=@n5, B=@n6)
   ) {materialFunction=&/MF_CiderPetalFlax_Crystal}.Result)
  n7 = FunctionInput() {inputName=Tech Art Mask}
  n8 = NamedRerouteDeclaration(None=ChannelMaskParameter(None=@n7) {group=CiderPetalFlax - Blue Crystal, parameterName=BlueCrystalMask.Channel, sortPriority=-1}.R) {name=BlueCrystalMask, nodeColor=<rgba 0.125 1 0 1>}
  n9 = ChannelMaskParameter(None=@n7) {group=CiderPetalFlax - Pink Crystal, maskChannel=Blue, parameterName=PinkCrystalMask.Channel, sortPriority=-1, defaultValue=<rgba 0 0 1 0>}
  n10 = NamedRerouteDeclaration(None=@n9.R) {name=PinkCrystalMask, nodeColor=<rgba 1 0 0.4453 1>}
  Max(A=@n8, B=@n10)
  NamedRerouteDeclaration(None=FunctionInput() {inputName=Material, inputType=FunctionInput_MaterialAttributes}) {name=IN_Material, nodeColor=<rgba 0.1953 0 1 1>}
  n11 = VectorParameter() {group=CiderPetalFlax - Pink Crystal, parameterName=PinkCrystal_ReflectionColor, sortPriority=50, defaultValue=<rgba 0.2 0 0.1 1>}
  n12 = VectorParameter() {group=CiderPetalFlax - Blue Crystal, parameterName=BlueCrystal_ReflectionColor, sortPriority=50, defaultValue=<rgba 0.0278 0.0347 0.1 1>}
  NamedRerouteDeclaration(None=LinearInterpolate(A=@n11.RGB, B=@n12.RGB, Alpha=NamedRerouteUsage())) {name=Crystal_ReflectionColor, nodeColor=<rgba 0 1 0.8594 1>}
  n13 = GetMaterialAttributes(None=NamedRerouteUsage()) {attributeGetTypes [#1]}
  n14 = NamedRerouteUsage()
  n15 = OneMinus(None=ScalarParameter() {defaultValue=5, group=CiderPetalFlax - Pink Crystal, parameterName=PinkCrystal_DetailNormal_Intensity, sortPriority=100})
  n16 = NamedRerouteUsage()
  n17 = OneMinus(None=ScalarParameter() {defaultValue=1, group=CiderPetalFlax - Blue Crystal, parameterName=BlueCrystal_DetailNormal_Intensity, sortPriority=100})
  n18 = MaterialFunctionCall(
   BaseNormal=@n13.Normal
   AdditionalNormal=LinearInterpolate(
    A=MaterialFunctionCall(Normal=@n14, Flatness=@n15) {materialFunction=&FlattenNormal}.Result
    B=MaterialFunctionCall(Normal=@n16, Flatness=@n17) {materialFunction=&FlattenNormal}.Result
    Alpha=NamedRerouteUsage()
   )
  ) {materialFunction=&BlendAngleCorrectedNormals}
  n19 = NamedRerouteUsage()
  NamedRerouteDeclaration(None=SetMaterialAttributes!(MaterialAttributes=@n13.MaterialAttributes, Normal=LinearInterpolate(A=@n13.Normal, B=@n18.Result, Alpha=@n19)) {attributeSetTypes [#1]}) {name=NormalMat, nodeColor=<rgba 1 0 0.4688 1>}
  NamedRerouteDeclaration(None=Saturate(None=Add(A=@n8, B=@n10))) {name=GlobalCrystalMask, nodeColor=<rgba 0 0.7656 1 1>}
  n20 = TextureSample(UVs=Multiply(A=TextureCoordinate(), B=ScalarParameter() {defaultValue=10, group=CiderPetalFlax - Global, parameterName=DetailNormal_Tile})) {samplerType=SAMPLERTYPE_Normal, texture=/BRCosmetics/Weapons/FORT_Melee/Pickaxe_CiderPetalFlax/Textures/T_CiderPetalFlax_CrystalNormal}
  NamedRerouteDeclaration(None=@n20.RGB) {name=DetailNormal, nodeColor=<rgba 1 0.9141 0 1>}
  n21 = VectorParameter() {group=CiderPetalFlax - Pink Crystal, parameterName=PinkCrystal_BaseColor, sortPriority=0, defaultValue=<rgba 0.15 0 0.1 1>}
  n22 = VectorParameter() {group=CiderPetalFlax - Blue Crystal, parameterName=BlueCrystal_BaseColor, sortPriority=0, defaultValue=<rgba 1 1 1 1>}
  NamedRerouteDeclaration(None=LinearInterpolate(A=@n21.RGB, B=@n22.RGB, Alpha=NamedRerouteUsage())) {name=Crystal_BaseColor, nodeColor=<rgba 0 1 0.8594 1>}
  NamedRerouteDeclaration(None=LinearInterpolate(
    A=ScalarParameter() {defaultValue=10, group=CiderPetalFlax - Pink Crystal, parameterName=PinkCrystal_InteriorIntensity, sortPriority=10}
    B=ScalarParameter() {defaultValue=3, group=CiderPetalFlax - Blue Crystal, parameterName=BlueCrystal_InteriorIntensity, sortPriority=10}
    Alpha=NamedRerouteUsage()
   )) {name=Crystal_InteriorIntensity, nodeColor=<rgba 1 0.5391 0 1>}
  NamedRerouteDeclaration(None=LinearInterpolate(
    A=ScalarParameter() {defaultValue=0.8, group=CiderPetalFlax - Pink Crystal, parameterName=PinkCrystal_InteriorContrast, sortPriority=11}
    B=ScalarParameter() {defaultValue=0.5, group=CiderPetalFlax - Blue Crystal, parameterName=BlueCrystal_InteriorContrast, sortPriority=11}
    Alpha=NamedRerouteUsage()
   )) {name=Crystal_InteriorContrast, nodeColor=<rgba 1 0.5391 0 1>}
  NamedRerouteDeclaration(None=LinearInterpolate(
    A=ScalarParameter() {defaultValue=1, group=CiderPetalFlax - Pink Crystal, parameterName=PinkCrystal_InteriorPrismColors, sortPriority=12}
    B=ScalarParameter() {defaultValue=1, group=CiderPetalFlax - Blue Crystal, parameterName=BlueCrystal_InteriorPrismColors, sortPriority=12}
    Alpha=NamedRerouteUsage()
   )) {name=Crystal_InteriorPrismColors, nodeColor=<rgba 1 0.5391 0 1>}
 ]}
