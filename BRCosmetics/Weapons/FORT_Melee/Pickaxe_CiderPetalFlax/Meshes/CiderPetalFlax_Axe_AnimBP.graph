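# Animation Blueprint for the CiderPetalFlax pickaxe that adds procedural secondary motion to its many loose pieces.
# From the cached 'FrontendDynamics' pose, chained Anim Dynamics nodes spring dyn_pieace bones within linear limits at varied alphas.
# They feed base_pieace nodes simulated relative to the Melee_TwoHanded_Handle bone, with boolean-driven, eased alpha blending and overridden gravity.
# A separate, unconnected dyn_pieace chain uses small angular limits instead.
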AnimBlueprint /BRCosmetics/Weapons/FORT_Melee/Pickaxe_CiderPetalFlax/Meshes/CiderPetalFlax_Axe_AnimBP
target_skeleton=&/CiderPetalFlax_Axe_Skeleton
graphs [
 AnimGraph {nodes [
   n1 = Anim Dynamics - Bone: dyn_pieace_01(Alpha=1, ExternalForce="0,0,0") {
    bGravityOverrideInSimSpace=true
    bLinearSpring=true
    gravityScale=0.07
    linearSpringConstant=10
    alphaBoolBlend {blendOption=Linear}
    alphaScaleBiasClamp {scale=0.11}
    boundBone {boneName=dyn_pieace_01}
    componentLinearAccScale=<0.11 0.13 0.18>
    physicsBodyDefinitions [{
      sphereCollisionRadius=0
      boundBone {boneName=dyn_pieace_01}
      boxExtents=<1 1 1>
      constraintSetup {angularLimitsMax=<0 10 10>, angularLimitsMin=<0 -10 -10>, angularTarget=<0 0 0>}
      localJointOffset=<0 0 0>
     }]
   }
   n2 = Anim Dynamics - Bone: dyn_pieace_02(ComponentPose=@n1, Alpha=1, ExternalForce="0,0,0") {
    bGravityOverrideInSimSpace=true
    bLinearSpring=true
    gravityScale=0.07
    linearSpringConstant=10
    alphaBoolBlend {blendOption=Linear}
    alphaScaleBiasClamp {scale=0.11}
    boundBone {boneName=dyn_pieace_02}
    componentLinearAccScale=<0.11 0.13 0.18>
    physicsBodyDefinitions [{
      sphereCollisionRadius=0
      boundBone {boneName=dyn_pieace_02}
      boxExtents=<1 1 1>
      constraintSetup {angularLimitsMax=<0 10 10>, angularLimitsMin=<0 -10 -10>, angularTarget=<0 0 0>}
      localJointOffset=<0 0 0>
     }]
   }
   n3 = Anim Dynamics - Bone: dyn_pieace_03(ComponentPose=@n2, Alpha=1, ExternalForce="0,0,0") {
    bGravityOverrideInSimSpace=true
    bLinearSpring=true
    gravityScale=0.07
    linearSpringConstant=10
    alphaBoolBlend {blendOption=Linear}
    alphaScaleBiasClamp {scale=0.11}
    boundBone {boneName=dyn_pieace_03}
    componentLinearAccScale=<0.11 0.13 0.2>
    physicsBodyDefinitions [{
      sphereCollisionRadius=0
      boundBone {boneName=dyn_pieace_03}
      boxExtents=<1 1 1>
      constraintSetup {angularLimitsMax=<0 10 10>, angularLimitsMin=<0 -10 -10>, angularTarget=<0 0 0>}
      localJointOffset=<0 0 0>
     }]
   }
   n4 = Anim Dynamics - Bone: dyn_pieace_04(ComponentPose=@n3, Alpha=1, ExternalForce="0,0,0") {
    bGravityOverrideInSimSpace=true
    bLinearSpring=true
    gravityScale=0.07
    linearSpringConstant=10
    alphaBoolBlend {blendOption=Linear}
    alphaScaleBiasClamp {scale=0.11}
    boundBone {boneName=dyn_pieace_04}
    componentLinearAccScale=<0.11 0.13 0.18>
    physicsBodyDefinitions [{
      sphereCollisionRadius=0
      boundBone {boneName=dyn_pieace_04}
      boxExtents=<1 1 1>
      constraintSetup {angularLimitsMax=<0 10 10>, angularLimitsMin=<0 -10 -10>, angularTarget=<0 0 0>}
      localJointOffset=<0 0 0>
     }]
   }
   n5 = Anim Dynamics - Bone: dyn_pieace_05(ComponentPose=@n4, Alpha=1, ExternalForce="0,0,0") {
    bGravityOverrideInSimSpace=true
    bLinearSpring=true
    gravityScale=0.07
    linearSpringConstant=10
    alphaBoolBlend {blendOption=Linear}
    alphaScaleBiasClamp {scale=0.11}
    boundBone {boneName=dyn_pieace_05}
    componentLinearAccScale=<0.11 0.13 0.18>
    physicsBodyDefinitions [{
      sphereCollisionRadius=0
      boundBone {boneName=dyn_pieace_05}
      boxExtents=<1 1 1>
      constraintSetup {angularLimitsMax=<0 10 10>, angularLimitsMin=<0 -10 -10>, angularTarget=<0 0 0>}
      localJointOffset=<0 0 0>
     }]
   }
   n6 = Anim Dynamics - Bone: dyn_pieace_06(ComponentPose=@n5, Alpha=1, ExternalForce="0,0,0") {
    bGravityOverrideInSimSpace=true
    bLinearSpring=true
    gravityScale=0.07
    linearSpringConstant=10
    alphaBoolBlend {blendOption=Linear}
    alphaScaleBiasClamp {scale=0.11}
    boundBone {boneName=dyn_pieace_06}
    componentLinearAccScale=<0.11 0.13 0.18>
    physicsBodyDefinitions [{
      sphereCollisionRadius=0
      boundBone {boneName=dyn_pieace_06}
      boxExtents=<1 1 1>
      constraintSetup {angularLimitsMax=<0 10 10>, angularLimitsMin=<0 -10 -10>, angularTarget=<0 0 0>}
      localJointOffset=<0 0 0>
     }]
   }
   n7 = Anim Dynamics - Bone: dyn_pieace_07(ComponentPose=@n6, Alpha=1, ExternalForce="0,0,0") {
    bGravityOverrideInSimSpace=true
    bLinearSpring=true
    gravityScale=0.07
    linearSpringConstant=10
    alphaBoolBlend {blendOption=Linear}
    alphaScaleBiasClamp {scale=0.11}
    boundBone {boneName=dyn_pieace_07}
    componentLinearAccScale=<0.11 0.13 0.21>
    physicsBodyDefinitions [{
      sphereCollisionRadius=0
      boundBone {boneName=dyn_pieace_07}
      boxExtents=<1 1 1>
      constraintSetup {angularLimitsMax=<0 10 10>, angularLimitsMin=<0 -10 -10>, angularTarget=<0 0 0>}
      localJointOffset=<0 0 0>
     }]
   }
   n8 = Anim Dynamics - Bone: dyn_pieace_08(ComponentPose=@n7, Alpha=1, ExternalForce="0,0,0") {
    bGravityOverrideInSimSpace=true
    bLinearSpring=true
    gravityScale=0.07
    linearSpringConstant=10
    alphaBoolBlend {blendOption=Linear}
    alphaScaleBiasClamp {scale=0.11}
    boundBone {boneName=dyn_pieace_08}
    componentLinearAccScale=<0.11 0.13 0.18>
    physicsBodyDefinitions [{
      sphereCollisionRadius=0
      boundBone {boneName=dyn_pieace_08}
      boxExtents=<1 1 1>
      constraintSetup {angularLimitsMax=<0 10 10>, angularLimitsMin=<0 -10 -10>, angularTarget=<0 0 0>}
      localJointOffset=<0 0 0>
     }]
   }
   n9 = Anim Dynamics - Bone: dyn_pieace_09(ComponentPose=@n8, Alpha=1, ExternalForce="0,0,0") {
    bGravityOverrideInSimSpace=true
    bLinearSpring=true
    gravityScale=0.07
    linearSpringConstant=10
    alphaBoolBlend {blendOption=Linear}
    alphaScaleBiasClamp {scale=0.11}
    boundBone {boneName=dyn_pieace_09}
    componentLinearAccScale=<0.11 0.13 0.21>
    physicsBodyDefinitions [{
      sphereCollisionRadius=0
      boundBone {boneName=dyn_pieace_09}
      boxExtents=<1 1 1>
      constraintSetup {angularLimitsMax=<0 10 10>, angularLimitsMin=<0 -10 -10>, angularTarget=<0 0 0>}
      localJointOffset=<0 0 0>
     }]
   }
   n10 = Anim Dynamics - Bone: dyn_pieace_10(ComponentPose=@n9, Alpha=1, ExternalForce="0,0,0") {
    bGravityOverrideInSimSpace=true
    bLinearSpring=true
    gravityScale=0.07
    linearSpringConstant=10
    alphaBoolBlend {blendOption=Linear}
    alphaScaleBiasClamp {scale=0.11}
    boundBone {boneName=dyn_pieace_10}
    componentLinearAccScale=<0.11 0.13 0.18>
    physicsBodyDefinitions [{
      sphereCollisionRadius=0
      boundBone {boneName=dyn_pieace_10}
      boxExtents=<1 1 1>
      constraintSetup {angularLimitsMax=<0 10 10>, angularLimitsMin=<0 -10 -10>, angularTarget=<0 0 0>}
      localJointOffset=<0 0 0>
     }]
   }
   n11 = Anim Dynamics - Bone: dyn_pieace_11(ComponentPose=@n10, Alpha=1, ExternalForce="0,0,0") {
    bGravityOverrideInSimSpace=true
    bLinearSpring=true
    gravityScale=0.07
    linearSpringConstant=10
    alphaBoolBlend {blendOption=Linear}
    alphaScaleBiasClamp {scale=0.11}
    boundBone {boneName=dyn_pieace_11}
    componentLinearAccScale=<0.11 0.13 0.18>
    physicsBodyDefinitions [{
      sphereCollisionRadius=0
      boundBone {boneName=dyn_pieace_11}
      boxExtents=<1 1 1>
      constraintSetup {angularLimitsMax=<0 10 10>, angularLimitsMin=<0 -10 -10>, angularTarget=<0 0 0>}
      localJointOffset=<0 0 0>
     }]
   }
   n12 = Anim Dynamics - Bone: dyn_pieace_12(ComponentPose=@n11, Alpha=1, ExternalForce="0,0,0") {
    bGravityOverrideInSimSpace=true
    bLinearSpring=true
    gravityScale=0.07
    linearSpringConstant=10
    alphaBoolBlend {blendOption=Linear}
    alphaScaleBiasClamp {scale=0.11}
    boundBone {boneName=dyn_pieace_12}
    componentLinearAccScale=<0.11 0.13 0.21>
    physicsBodyDefinitions [{
      sphereCollisionRadius=0
      boundBone {boneName=dyn_pieace_12}
      boxExtents=<1 1 1>
      constraintSetup {angularLimitsMax=<0 10 10>, angularLimitsMin=<0 -10 -10>, angularTarget=<0 0 0>}
      localJointOffset=<0 0 0>
     }]
   }
   n13 = Anim Dynamics - Bone: dyn_pieace_13(ComponentPose=@n12, Alpha=1, ExternalForce="0,0,0") {
    bGravityOverrideInSimSpace=true
    bLinearSpring=true
    gravityScale=0.07
    linearSpringConstant=10
    alphaBoolBlend {blendOption=Linear}
    alphaScaleBiasClamp {scale=0.11}
    boundBone {boneName=dyn_pieace_13}
    componentLinearAccScale=<0.11 0.13 0.18>
    physicsBodyDefinitions [{
      sphereCollisionRadius=0
      boundBone {boneName=dyn_pieace_13}
      boxExtents=<1 1 1>
      constraintSetup {angularLimitsMax=<0 10 10>, angularLimitsMin=<0 -10 -10>, angularTarget=<0 0 0>}
      localJointOffset=<0 0 0>
     }]
   }
   n14 = Anim Dynamics - Bone: dyn_pieace_14(ComponentPose=@n13, Alpha=1, ExternalForce="0,0,0") {
    bGravityOverrideInSimSpace=true
    bLinearSpring=true
    gravityScale=0.07
    linearSpringConstant=10
    alphaBoolBlend {blendOption=Linear}
    alphaScaleBiasClamp {scale=0.11}
    boundBone {boneName=dyn_pieace_14}
    componentLinearAccScale=<0.11 0.13 0.21>
    physicsBodyDefinitions [{
      sphereCollisionRadius=0
      boundBone {boneName=dyn_pieace_14}
      boxExtents=<1 1 1>
      constraintSetup {angularLimitsMax=<0 10 10>, angularLimitsMin=<0 -10 -10>, angularTarget=<0 0 0>}
      localJointOffset=<0 0 0>
     }]
   }
   n15 = Anim Dynamics - Bone: dyn_pieace_15(ComponentPose=@n14, Alpha=1, ExternalForce="0,0,0") {
    bGravityOverrideInSimSpace=true
    bLinearSpring=true
    gravityScale=0.07
    linearSpringConstant=10
    alphaBoolBlend {blendOption=Linear}
    alphaScaleBiasClamp {scale=0.11}
    boundBone {boneName=dyn_pieace_15}
    componentLinearAccScale=<0.11 0.13 0.18>
    physicsBodyDefinitions [{
      sphereCollisionRadius=0
      boundBone {boneName=dyn_pieace_15}
      boxExtents=<1 1 1>
      constraintSetup {angularLimitsMax=<0 10 10>, angularLimitsMin=<0 -10 -10>, angularTarget=<0 0 0>}
      localJointOffset=<0 0 0>
     }]
   }
   n16 = Anim Dynamics - Bone: dyn_pieace_16(ComponentPose=@n15, Alpha=1, ExternalForce="0,0,0") {
    bGravityOverrideInSimSpace=true
    bLinearSpring=true
    gravityScale=0.07
    linearSpringConstant=10
    alphaBoolBlend {blendOption=Linear}
    alphaScaleBiasClamp {scale=0.11}
    boundBone {boneName=dyn_pieace_16}
    componentLinearAccScale=<0.11 0.13 0.18>
    physicsBodyDefinitions [{
      sphereCollisionRadius=0
      boundBone {boneName=dyn_pieace_16}
      boxExtents=<1 1 1>
      constraintSetup {angularLimitsMax=<0 10 10>, angularLimitsMin=<0 -10 -10>, angularTarget=<0 0 0>}
      localJointOffset=<0 0 0>
     }]
   }
   n17 = Anim Dynamics - Bone: dyn_pieace_17(ComponentPose=@n16, Alpha=1, ExternalForce="0,0,0") {
    bGravityOverrideInSimSpace=true
    bLinearSpring=true
    gravityScale=0.07
    linearSpringConstant=10
    alphaBoolBlend {blendOption=Linear}
    alphaScaleBiasClamp {scale=0.11}
    boundBone {boneName=dyn_pieace_17}
    componentLinearAccScale=<0.11 0.13 0.21>
    physicsBodyDefinitions [{
      sphereCollisionRadius=0
      boundBone {boneName=dyn_pieace_17}
      boxExtents=<1 1 1>
      constraintSetup {angularLimitsMax=<0 10 10>, angularLimitsMin=<0 -10 -10>, angularTarget=<0 0 0>}
      localJointOffset=<0 0 0>
     }]
   }
   n18 = Anim Dynamics - Bone: dyn_pieace_18(ComponentPose=@n17, Alpha=1, ExternalForce="0,0,0") {
    bGravityOverrideInSimSpace=true
    bLinearSpring=true
    gravityScale=0.07
    linearSpringConstant=10
    alphaBoolBlend {blendOption=Linear}
    alphaScaleBiasClamp {scale=0.11}
    boundBone {boneName=dyn_pieace_18}
    componentLinearAccScale=<0.11 0.13 0.18>
    physicsBodyDefinitions [{
      sphereCollisionRadius=0
      boundBone {boneName=dyn_pieace_18}
      boxExtents=<1 1 1>
      constraintSetup {angularLimitsMax=<0 10 10>, angularLimitsMin=<0 -10 -10>, angularTarget=<0 0 0>}
      localJointOffset=<0 0 0>
     }]
   }
   n19 = Anim Dynamics - Bone: dyn_pieace_19(ComponentPose=@n18, Alpha=1, ExternalForce="0,0,0") {
    bGravityOverrideInSimSpace=true
    bLinearSpring=true
    gravityScale=0.07
    linearSpringConstant=10
    alphaBoolBlend {blendOption=Linear}
    alphaScaleBiasClamp {scale=0.11}
    boundBone {boneName=dyn_pieace_19}
    componentLinearAccScale=<0.11 0.13 0.21>
    physicsBodyDefinitions [{
      sphereCollisionRadius=0
      boundBone {boneName=dyn_pieace_19}
      boxExtents=<1 1 1>
      constraintSetup {angularLimitsMax=<0 10 10>, angularLimitsMin=<0 -10 -10>, angularTarget=<0 0 0>}
      localJointOffset=<0 0 0>
     }]
   }
   Anim Dynamics - Bone: dyn_pieace_20(ComponentPose=@n19, Alpha=1, ExternalForce="0,0,0") {
    bGravityOverrideInSimSpace=true
    bLinearSpring=true
    gravityScale=0.07
    linearSpringConstant=10
    alphaBoolBlend {blendOption=Linear}
    alphaScaleBiasClamp {scale=0.11}
    boundBone {boneName=dyn_pieace_20}
    componentLinearAccScale=<0.11 0.13 0.18>
    physicsBodyDefinitions [{
      sphereCollisionRadius=0
      boundBone {boneName=dyn_pieace_20}
      boxExtents=<1 1 1>
      constraintSetup {angularLimitsMax=<0 10 10>, angularLimitsMin=<0 -10 -10>, angularTarget=<0 0 0>}
      localJointOffset=<0 0 0>
     }]
   }
   n20 = Anim Dynamics - Bone: dyn_pieace_01(ComponentPose=Local To Component(LocalPose=Use cached pose 'FrontendDynamics'()), Alpha=1, ExternalForce="0,0,0") {
    bGravityOverrideInSimSpace=true
    bLinearSpring=true
    gravityScale=0.07
    linearSpringConstant=10
    alphaBoolBlend {blendOption=Linear}
    alphaScaleBiasClamp {scale=0.11}
    boundBone {boneName=dyn_pieace_01}
    componentLinearAccScale=<0.11 0.13 0.18>
    physicsBodyDefinitions [{
      sphereCollisionRadius=0
      boundBone {boneName=dyn_pieace_01}
      boxExtents=<1 1 1>
      constraintSetup {angularLimitsMax=<0 0 0>, angularLimitsMin=<0 0 0>, angularTarget=<0 0 0>, linearAxesMax=<50 40 30>, linearAxesMin=<-60 0 -60>}
      localJointOffset=<0 0 0>
     }]
   }
   n21 = Anim Dynamics - Bone: dyn_pieace_02(ComponentPose=@n20, Alpha=0.5, ExternalForce="0,0,0") {
    bGravityOverrideInSimSpace=true
    bLinearSpring=true
    gravityScale=0.07
    linearSpringConstant=10
    alphaBoolBlend {blendOption=Linear}
    alphaScaleBiasClamp {scale=0.11}
    boundBone {boneName=dyn_pieace_02}
    componentLinearAccScale=<0.11 0.13 0.18>
    physicsBodyDefinitions [{
      sphereCollisionRadius=0
      boundBone {boneName=dyn_pieace_02}
      boxExtents=<1 1 1>
      constraintSetup {angularLimitsMax=<0 0 0>, angularLimitsMin=<0 0 0>, angularTarget=<0 0 0>, linearAxesMax=<50 30 30>, linearAxesMin=<-60 -60 -60>}
      localJointOffset=<0 0 0>
     }]
   }
   n22 = Anim Dynamics - Bone: dyn_pieace_03(ComponentPose=@n21, Alpha=0.7, ExternalForce="0,0,0") {
    bGravityOverrideInSimSpace=true
    bLinearSpring=true
    gravityScale=0.07
    linearSpringConstant=10
    alphaBoolBlend {blendOption=Linear}
    alphaScaleBiasClamp {scale=0.11}
    boundBone {boneName=dyn_pieace_03}
    componentLinearAccScale=<0.11 0.13 0.2>
    physicsBodyDefinitions [{
      sphereCollisionRadius=0
      boundBone {boneName=dyn_pieace_03}
      boxExtents=<1 1 1>
      constraintSetup {angularLimitsMax=<0 0 0>, angularLimitsMin=<0 0 0>, angularTarget=<0 0 0>, linearAxesMax=<60 30 30>, linearAxesMin=<-60 -60 -60>}
      localJointOffset=<0 0 0>
     }]
   }
   n23 = Anim Dynamics - Bone: dyn_pieace_04(ComponentPose=@n22, Alpha=1, ExternalForce="0,0,0") {
    bGravityOverrideInSimSpace=true
    bLinearSpring=true
    gravityScale=0.07
    linearSpringConstant=10
    alphaBoolBlend {blendOption=Linear}
    alphaScaleBiasClamp {scale=0.11}
    boundBone {boneName=dyn_pieace_04}
    componentLinearAccScale=<0.11 0.13 0.18>
    physicsBodyDefinitions [{
      sphereCollisionRadius=0
      boundBone {boneName=dyn_pieace_04}
      boxExtents=<1 1 1>
      constraintSetup {angularLimitsMax=<0 0 0>, angularLimitsMin=<0 0 0>, angularTarget=<0 0 0>, linearAxesMax=<50 30 30>, linearAxesMin=<-60 -60 -60>}
      localJointOffset=<0 0 0>
     }]
   }
   n24 = Anim Dynamics - Bone: dyn_pieace_05(ComponentPose=@n23, Alpha=1, ExternalForce="0,0,0") {
    bGravityOverrideInSimSpace=true
    bLinearSpring=true
    gravityScale=0.07
    linearSpringConstant=10
    alphaBoolBlend {blendOption=Linear}
    alphaScaleBiasClamp {scale=0.11}
    boundBone {boneName=dyn_pieace_05}
    componentLinearAccScale=<0.11 0.13 0.18>
    physicsBodyDefinitions [{
      sphereCollisionRadius=0
      boundBone {boneName=dyn_pieace_05}
      boxExtents=<1 1 1>
      constraintSetup {angularLimitsMax=<0 0 0>, angularLimitsMin=<0 0 0>, angularTarget=<0 0 0>, linearAxesMax=<50 40 30>, linearAxesMin=<-60 0 -60>}
      localJointOffset=<0 0 0>
     }]
   }
   n25 = Anim Dynamics - Bone: dyn_pieace_06(ComponentPose=@n24, Alpha=0.3, ExternalForce="0,0,0") {
    bGravityOverrideInSimSpace=true
    bLinearSpring=true
    gravityScale=0.07
    linearSpringConstant=10
    alphaBoolBlend {blendOption=Linear}
    alphaScaleBiasClamp {scale=0.11}
    boundBone {boneName=dyn_pieace_06}
    componentLinearAccScale=<0.11 0.13 0.18>
    physicsBodyDefinitions [{
      sphereCollisionRadius=0
      boundBone {boneName=dyn_pieace_06}
      boxExtents=<1 1 1>
      constraintSetup {angularLimitsMax=<0 0 0>, angularLimitsMin=<0 0 0>, angularTarget=<0 0 0>, linearAxesMax=<20 30 30>, linearAxesMin=<-70 -60 -60>}
      localJointOffset=<0 0 0>
     }]
   }
   n26 = Anim Dynamics - Bone: dyn_pieace_07(ComponentPose=@n25, Alpha=1, ExternalForce="0,0,0") {
    bGravityOverrideInSimSpace=true
    bLinearSpring=true
    gravityScale=0.07
    linearSpringConstant=10
    alphaBoolBlend {blendOption=Linear}
    alphaScaleBiasClamp {scale=0.11}
    boundBone {boneName=dyn_pieace_07}
    componentLinearAccScale=<0.11 0.13 0.21>
    physicsBodyDefinitions [{
      sphereCollisionRadius=0
      boundBone {boneName=dyn_pieace_07}
      boxExtents=<1 1 1>
      constraintSetup {angularLimitsMax=<0 0 0>, angularLimitsMin=<0 0 0>, angularTarget=<0 0 0>, linearAxesMax=<50 30 30>, linearAxesMin=<-60 -60 -60>}
      localJointOffset=<0 0 0>
     }]
   }
   n27 = Anim Dynamics - Bone: dyn_pieace_08(ComponentPose=@n26, Alpha=1, ExternalForce="0,0,0") {
    bGravityOverrideInSimSpace=true
    bLinearSpring=true
    gravityScale=0.07
    linearSpringConstant=10
    alphaBoolBlend {blendOption=Linear}
    alphaScaleBiasClamp {scale=0.11}
    boundBone {boneName=dyn_pieace_08}
    componentLinearAccScale=<0.11 0.13 0.18>
    physicsBodyDefinitions [{
      sphereCollisionRadius=0
      boundBone {boneName=dyn_pieace_08}
      boxExtents=<1 1 1>
      constraintSetup {angularLimitsMax=<0 0 0>, angularLimitsMin=<0 0 0>, angularTarget=<0 0 0>, linearAxesMax=<30 30 30>, linearAxesMin=<-60 -60 -60>}
      localJointOffset=<0 0 0>
     }]
   }
   n28 = Anim Dynamics - Bone: dyn_pieace_09(ComponentPose=@n27, Alpha=0.2, ExternalForce="0,0,0") {
    bGravityOverrideInSimSpace=true
    bLinearSpring=true
    gravityScale=0.07
    linearSpringConstant=10
    alphaBoolBlend {blendOption=Linear}
    alphaScaleBiasClamp {scale=0.11}
    boundBone {boneName=dyn_pieace_09}
    componentLinearAccScale=<0.11 0.13 0.21>
    physicsBodyDefinitions [{
      sphereCollisionRadius=0
      boundBone {boneName=dyn_pieace_09}
      boxExtents=<1 1 1>
      constraintSetup {angularLimitsMax=<0 0 0>, angularLimitsMin=<0 0 0>, angularTarget=<0 0 0>, linearAxesMax=<50 30 30>, linearAxesMin=<-30 -30 -30>}
      localJointOffset=<0 0 0>
     }]
   }
   n29 = Anim Dynamics - Bone: dyn_pieace_10(ComponentPose=@n28, Alpha=0.5, ExternalForce="0,0,0") {
    bGravityOverrideInSimSpace=true
    bLinearSpring=true
    gravityScale=0.07
    linearSpringConstant=10
    alphaBoolBlend {blendOption=Linear}
    alphaScaleBiasClamp {scale=0.11}
    boundBone {boneName=dyn_pieace_10}
    componentLinearAccScale=<0.11 0.13 0.18>
    physicsBodyDefinitions [{
      sphereCollisionRadius=0
      boundBone {boneName=dyn_pieace_10}
      boxExtents=<1 1 1>
      constraintSetup {angularLimitsMax=<0 0 0>, angularLimitsMin=<0 0 0>, angularTarget=<0 0 0>, linearAxesMax=<30 40 30>, linearAxesMin=<-60 0 -60>}
      localJointOffset=<0 0 0>
     }]
   }
   n30 = Anim Dynamics - Bone: dyn_pieace_11(ComponentPose=@n29, Alpha=1, ExternalForce="0,0,0") {
    bGravityOverrideInSimSpace=true
    bLinearSpring=true
    gravityScale=0.07
    linearSpringConstant=10
    alphaBoolBlend {blendOption=Linear}
    alphaScaleBiasClamp {scale=0.11}
    boundBone {boneName=dyn_pieace_11}
    componentLinearAccScale=<0.11 0.13 0.18>
    physicsBodyDefinitions [{
      sphereCollisionRadius=0
      boundBone {boneName=dyn_pieace_11}
      boxExtents=<1 1 1>
      constraintSetup {angularLimitsMax=<0 0 0>, angularLimitsMin=<0 0 0>, angularTarget=<0 0 0>, linearAxesMax=<50 30 30>, linearAxesMin=<-60 -60 -60>}
      localJointOffset=<0 0 0>
     }]
   }
   n31 = Anim Dynamics - Bone: dyn_pieace_12(ComponentPose=@n30, Alpha=1, ExternalForce="0,0,0") {
    bGravityOverrideInSimSpace=true
    bLinearSpring=true
    gravityScale=0.07
    linearSpringConstant=10
    alphaBoolBlend {blendOption=Linear}
    alphaScaleBiasClamp {scale=0.11}
    boundBone {boneName=dyn_pieace_12}
    componentLinearAccScale=<0.11 0.13 0.21>
    physicsBodyDefinitions [{
      sphereCollisionRadius=0
      boundBone {boneName=dyn_pieace_12}
      boxExtents=<1 1 1>
      constraintSetup {angularLimitsMax=<0 0 0>, angularLimitsMin=<0 0 0>, angularTarget=<0 0 0>, linearAxesMax=<50 30 30>, linearAxesMin=<-60 -60 -60>}
      localJointOffset=<0 0 0>
     }]
   }
   n32 = Anim Dynamics - Bone: dyn_pieace_13(ComponentPose=@n31, Alpha=1, ExternalForce="0,0,0") {
    bGravityOverrideInSimSpace=true
    bLinearSpring=true
    gravityScale=0.07
    linearSpringConstant=10
    alphaBoolBlend {blendOption=Linear}
    alphaScaleBiasClamp {scale=0.11}
    boundBone {boneName=dyn_pieace_13}
    componentLinearAccScale=<0.11 0.13 0.18>
    physicsBodyDefinitions [{
      sphereCollisionRadius=0
      boundBone {boneName=dyn_pieace_13}
      boxExtents=<1 1 1>
      constraintSetup {angularLimitsMax=<0 0 0>, angularLimitsMin=<0 0 0>, angularTarget=<0 0 0>, linearAxesMax=<50 30 30>, linearAxesMin=<-60 -60 -60>}
      localJointOffset=<0 0 0>
     }]
   }
   n33 = Anim Dynamics - Bone: dyn_pieace_14(ComponentPose=@n32, Alpha=1, ExternalForce="0,0,0") {
    bGravityOverrideInSimSpace=true
    bLinearSpring=true
    gravityScale=0.07
    linearSpringConstant=10
    alphaBoolBlend {blendOption=Linear}
    alphaScaleBiasClamp {scale=0.11}
    boundBone {boneName=dyn_pieace_14}
    componentLinearAccScale=<0.11 0.13 0.21>
    physicsBodyDefinitions [{
      sphereCollisionRadius=0
      boundBone {boneName=dyn_pieace_14}
      boxExtents=<1 1 1>
      constraintSetup {angularLimitsMax=<0 0 0>, angularLimitsMin=<0 0 0>, angularTarget=<0 0 0>, linearAxesMax=<50 30 30>, linearAxesMin=<-60 -60 -60>}
      localJointOffset=<0 0 0>
     }]
   }
   n34 = Anim Dynamics - Bone: dyn_pieace_15(ComponentPose=@n33, Alpha=1, ExternalForce="0,0,0") {
    bGravityOverrideInSimSpace=true
    bLinearSpring=true
    gravityScale=0.07
    linearSpringConstant=10
    alphaBoolBlend {blendOption=Linear}
    alphaScaleBiasClamp {scale=0.11}
    boundBone {boneName=dyn_pieace_15}
    componentLinearAccScale=<0.11 0.13 0.18>
    physicsBodyDefinitions [{
      sphereCollisionRadius=0
      boundBone {boneName=dyn_pieace_15}
      boxExtents=<1 1 1>
      constraintSetup {angularLimitsMax=<0 0 0>, angularLimitsMin=<0 0 0>, angularTarget=<0 0 0>, linearAxesMax=<30 40 30>, linearAxesMin=<-60 0 -60>}
      localJointOffset=<0 0 0>
     }]
   }
   n35 = Anim Dynamics - Bone: dyn_pieace_16(ComponentPose=@n34, Alpha=1, ExternalForce="0,0,0") {
    bGravityOverrideInSimSpace=true
    bLinearSpring=true
    gravityScale=0.07
    linearSpringConstant=10
    alphaBoolBlend {blendOption=Linear}
    alphaScaleBiasClamp {scale=0.11}
    boundBone {boneName=dyn_pieace_16}
    componentLinearAccScale=<0.11 0.13 0.18>
    physicsBodyDefinitions [{
      sphereCollisionRadius=0
      boundBone {boneName=dyn_pieace_16}
      boxExtents=<1 1 1>
      constraintSetup {angularLimitsMax=<0 0 0>, angularLimitsMin=<0 0 0>, angularTarget=<0 0 0>, linearAxesMax=<50 30 30>, linearAxesMin=<-60 -60 -60>}
      localJointOffset=<0 0 0>
     }]
   }
   n36 = Anim Dynamics - Bone: dyn_pieace_17(ComponentPose=@n35, Alpha=1, ExternalForce="0,0,0") {
    bGravityOverrideInSimSpace=true
    bLinearSpring=true
    gravityScale=0.07
    linearSpringConstant=10
    alphaBoolBlend {blendOption=Linear}
    alphaScaleBiasClamp {scale=0.11}
    boundBone {boneName=dyn_pieace_17}
    componentLinearAccScale=<0.11 0.13 0.21>
    physicsBodyDefinitions [{
      sphereCollisionRadius=0
      boundBone {boneName=dyn_pieace_17}
      boxExtents=<1 1 1>
      constraintSetup {angularLimitsMax=<0 0 0>, angularLimitsMin=<0 0 0>, angularTarget=<0 0 0>, linearAxesMax=<50 30 30>, linearAxesMin=<-60 -60 -60>}
      localJointOffset=<0 0 0>
     }]
   }
   n37 = Anim Dynamics - Bone: dyn_pieace_18(ComponentPose=@n36, Alpha=1, ExternalForce="0,0,0") {
    bGravityOverrideInSimSpace=true
    bLinearSpring=true
    gravityScale=0.07
    linearSpringConstant=10
    alphaBoolBlend {blendOption=Linear}
    alphaScaleBiasClamp {scale=0.11}
    boundBone {boneName=dyn_pieace_18}
    componentLinearAccScale=<0.11 0.13 0.18>
    physicsBodyDefinitions [{
      sphereCollisionRadius=0
      boundBone {boneName=dyn_pieace_18}
      boxExtents=<1 1 1>
      constraintSetup {angularLimitsMax=<0 0 0>, angularLimitsMin=<0 0 0>, angularTarget=<0 0 0>, linearAxesMax=<50 30 30>, linearAxesMin=<-60 -60 -60>}
      localJointOffset=<0 0 0>
     }]
   }
   n38 = Anim Dynamics - Bone: dyn_pieace_19(ComponentPose=@n37, Alpha=1, ExternalForce="0,0,0") {
    bGravityOverrideInSimSpace=true
    bLinearSpring=true
    gravityScale=0.07
    linearSpringConstant=10
    alphaBoolBlend {blendOption=Linear}
    alphaScaleBiasClamp {scale=0.11}
    boundBone {boneName=dyn_pieace_19}
    componentLinearAccScale=<0.11 0.13 0.21>
    physicsBodyDefinitions [{
      sphereCollisionRadius=0
      boundBone {boneName=dyn_pieace_19}
      boxExtents=<1 1 1>
      constraintSetup {angularLimitsMax=<0 0 0>, angularLimitsMin=<0 0 0>, angularTarget=<0 0 0>, linearAxesMax=<50 30 30>, linearAxesMin=<-60 -60 -60>}
      localJointOffset=<0 0 0>
     }]
   }
   n39 = Anim Dynamics - Bone: dyn_pieace_20(ComponentPose=@n38, Alpha=0.2, ExternalForce="0,0,0") {
    bGravityOverrideInSimSpace=true
    bLinearSpring=true
    gravityScale=0.07
    linearSpringConstant=10
    alphaBoolBlend {blendOption=Linear}
    alphaScaleBiasClamp {scale=0.11}
    boundBone {boneName=dyn_pieace_20}
    componentLinearAccScale=<0.11 0.13 0.18>
    physicsBodyDefinitions [{
      sphereCollisionRadius=0
      boundBone {boneName=dyn_pieace_20}
      boxExtents=<1 1 1>
      constraintSetup {angularLimitsMax=<0 0 0>, angularLimitsMin=<0 0 0>, angularTarget=<0 0 0>, linearAxesMax=<50 40 30>, linearAxesMin=<-60 0 -60>}
      localJointOffset=<0 0 0>
     }]
   }
   n40 = Anim Dynamics - Bone: base_pieace_01!(SimulationSpace:AnimPhysSimSpaceType Enum=BoneRelative, ComponentPose=@n39, bAlphaBoolEnabled:bool=True, ExternalForce="0,0,0") -> (Pose) {
    alphaInputType=Bool
    angularDampingOverride=0.9
    angularSpringConstant=11000
    bGravityOverrideInSimSpace=true
    bOverrideAngularDamping=true
    gravityScale=0.1
    linearSpringConstant=10
    simulationSpace=BoneRelative
    alphaBoolBlend {blendInTime=0.5, blendOption=QuadraticInOut, blendOutTime=1}
    alphaScaleBiasClamp {scale=0.11}
    boundBone {boneName=base_pieace_01}
    componentAppliedLinearAccClamp=<9500 9500 9500>
    componentLinearAccScale=<0.5 0.5 0.5>
    componentLinearVelScale=<0.5 0.5 0.5>
    gravityOverride {x=-1200}
    physicsBodyDefinitions [{
      sphereCollisionRadius=0
      boundBone {boneName=base_pieace_01}
      boxExtents=<44 44 44>
      constraintSetup {angularLimitsMax=<0 0 0>, angularLimitsMin=<0 0 0>, angularTarget=<0 0 0>, linearAxesMax=<7 1 1>, linearAxesMin=<-1 -1 -1>}
      localJointOffset=<1 0 0>
     }]
    relativeSpaceBone {boneName=Melee_TwoHanded_Handle}
    simSpaceSettings {maxAngularAcceleration=8, maxAngularVelocity=2.5, simSpaceAngularAlpha=1}
   }
   n41 = Anim Dynamics - Bone: base_pieace_02!(SimulationSpace:AnimPhysSimSpaceType Enum=BoneRelative, ComponentPose=@n40, bAlphaBoolEnabled:bool=True, ExternalForce="0,0,0") -> (Pose) {
    alphaInputType=Bool
    angularDampingOverride=0.9
    angularSpringConstant=11000
    bGravityOverrideInSimSpace=true
    bOverrideAngularDamping=true
    gravityScale=0.1
    linearSpringConstant=10
    simulationSpace=BoneRelative
    alphaBoolBlend {blendInTime=0.5, blendOption=QuadraticInOut, blendOutTime=1}
    alphaScaleBiasClamp {scale=0.11}
    boundBone {boneName=base_pieace_02}
    componentAppliedLinearAccClamp=<9500 9500 9500>
    componentLinearAccScale=<0.5 0.5 0.5>
    componentLinearVelScale=<0.5 0.5 0.5>
    gravityOverride {x=-1200}
    physicsBodyDefinitions [{
      sphereCollisionRadius=0
      boundBone {boneName=base_pieace_02}
      boxExtents=<45 45 45>
      constraintSetup {angularLimitsMax=<0 0 0>, angularLimitsMin=<0 0 0>, angularTarget=<0 0 0>, linearAxesMax=<8 1 1>, linearAxesMin=<-1 -1 -1>}
      localJointOffset=<1.4 0 0>
     }]
    relativeSpaceBone {boneName=Melee_TwoHanded_Handle}
    simSpaceSettings {maxAngularAcceleration=8, maxAngularVelocity=2.5, simSpaceAngularAlpha=1}
   }
   n42 = Anim Dynamics - Bone: base_pieace_03!(SimulationSpace:AnimPhysSimSpaceType Enum=BoneRelative, ComponentPose=@n41, bAlphaBoolEnabled:bool=True, ExternalForce="0,0,0") -> (Pose) {
    alphaInputType=Bool
    angularDampingOverride=0.9
    angularSpringConstant=11000
    bGravityOverrideInSimSpace=true
    bOverrideAngularDamping=true
    gravityScale=0.1
    linearSpringConstant=10
    simulationSpace=BoneRelative
    alphaBoolBlend {blendInTime=0.5, blendOption=QuadraticInOut, blendOutTime=1}
    alphaScaleBiasClamp {scale=0.11}
    boundBone {boneName=base_pieace_03}
    componentAppliedLinearAccClamp=<9500 9500 9500>
    componentLinearAccScale=<0.5 0.5 0.5>
    componentLinearVelScale=<0.5 0.5 0.5>
    gravityOverride {x=-1200}
    physicsBodyDefinitions [{
      sphereCollisionRadius=0
      boundBone {boneName=base_pieace_03}
      boxExtents=<35 35 35>
      constraintSetup {angularLimitsMax=<0 0 0>, angularLimitsMin=<0 0 0>, angularTarget=<0 0 0>, linearAxesMax=<13 1 1>, linearAxesMin=<-1 -1 -1>}
      localJointOffset=<2 0 0>
     }]
    relativeSpaceBone {boneName=Melee_TwoHanded_Handle}
    simSpaceSettings {maxAngularAcceleration=8, maxAngularVelocity=2.5, simSpaceAngularAlpha=1}
   }
   n43 = Anim Dynamics - Bone: base_pieace_04!(SimulationSpace:AnimPhysSimSpaceType Enum=BoneRelative, ComponentPose=@n42, bAlphaBoolEnabled:bool=True, ExternalForce="0,0,0") -> (Pose) {
    alphaInputType=Bool
    angularDampingOverride=0.9
    angularSpringConstant=11000
    bGravityOverrideInSimSpace=true
    bOverrideAngularDamping=true
    gravityScale=0.1
    linearSpringConstant=10
    simulationSpace=BoneRelative
    alphaBoolBlend {blendInTime=0.5, blendOption=QuadraticInOut, blendOutTime=1}
    alphaScaleBiasClamp {scale=0.11}
    boundBone {boneName=base_pieace_04}
    componentAppliedLinearAccClamp=<9500 9500 9500>
    componentLinearAccScale=<0.5 0.5 0.5>
    componentLinearVelScale=<0.5 0.5 0.5>
    gravityOverride {x=-1200}
    physicsBodyDefinitions [{
      sphereCollisionRadius=0
      boundBone {boneName=base_pieace_04}
      boxExtents=<38 38 38>
      constraintSetup {angularLimitsMax=<0 0 0>, angularLimitsMin=<0 0 0>, angularTarget=<0 0 0>, linearAxesMax=<11 1 1>, linearAxesMin=<-1 -1 -1>}
      localJointOffset=<1.5 0 0>
     }]
    relativeSpaceBone {boneName=Melee_TwoHanded_Handle}
    simSpaceSettings {maxAngularAcceleration=8, maxAngularVelocity=2.5, simSpaceAngularAlpha=1}
   }
   n44 = Anim Dynamics - Bone: base_pieace_05!(SimulationSpace:AnimPhysSimSpaceType Enum=BoneRelative, ComponentPose=@n43, bAlphaBoolEnabled:bool=True, ExternalForce="0,0,0") -> (Pose) {
    alphaInputType=Bool
    angularDampingOverride=0.9
    angularSpringConstant=11000
    bGravityOverrideInSimSpace=true
    bOverrideAngularDamping=true
    gravityScale=0.1
    linearSpringConstant=10
    simulationSpace=BoneRelative
    alphaBoolBlend {blendInTime=0.5, blendOption=QuadraticInOut, blendOutTime=1}
    alphaScaleBiasClamp {scale=0.11}
    boundBone {boneName=base_pieace_05}
    componentAppliedLinearAccClamp=<9500 9500 9500>
    componentLinearAccScale=<0.5 0.5 0.5>
    componentLinearVelScale=<0.5 0.5 0.5>
    gravityOverride {x=-1200}
    physicsBodyDefinitions [{
      sphereCollisionRadius=0
      boundBone {boneName=base_pieace_05}
      boxExtents=<40 40 40>
      constraintSetup {angularLimitsMax=<0 0 0>, angularLimitsMin=<0 0 0>, angularTarget=<0 0 0>, linearAxesMax=<12 1 1>, linearAxesMin=<-1 -1 -1>}
      localJointOffset=<2 0 0>
     }]
    relativeSpaceBone {boneName=Melee_TwoHanded_Handle}
    simSpaceSettings {maxAngularAcceleration=8, maxAngularVelocity=2.5, simSpaceAngularAlpha=1}
   }
   n45 = Anim Dynamics - Bone: base_pieace_06!(SimulationSpace:AnimPhysSimSpaceType Enum=BoneRelative, ComponentPose=@n44, bAlphaBoolEnabled:bool=True, ExternalForce="0,0,0") -> (Pose) {
    alphaInputType=Bool
    angularDampingOverride=0.9
    angularSpringConstant=11000
    bGravityOverrideInSimSpace=true
    bOverrideAngularDamping=true
    gravityScale=0.1
    linearSpringConstant=10
    simulationSpace=BoneRelative
    alphaBoolBlend {blendInTime=0.5, blendOption=QuadraticInOut, blendOutTime=1}
    alphaScaleBiasClamp {scale=0.11}
    boundBone {boneName=base_pieace_06}
    componentAppliedLinearAccClamp=<9500 9500 9500>
    componentLinearAccScale=<0.5 0.5 0.5>
    componentLinearVelScale=<0.5 0.5 0.5>
    gravityOverride {x=-1200}
    physicsBodyDefinitions [{
      sphereCollisionRadius=0
      boundBone {boneName=base_pieace_06}
      boxExtents=<38 38 38>
      constraintSetup {angularLimitsMax=<0 0 0>, angularLimitsMin=<0 0 0>, angularTarget=<0 0 0>, linearAxesMax=<2 1 1>, linearAxesMin=<-6 -1 -1>}
      localJointOffset=<2 0 0>
     }]
    relativeSpaceBone {boneName=Melee_TwoHanded_Handle}
    simSpaceSettings {maxAngularAcceleration=8, maxAngularVelocity=2.5, simSpaceAngularAlpha=1}
   }
   n46 = Anim Dynamics - Bone: base_pieace_07!(SimulationSpace:AnimPhysSimSpaceType Enum=BoneRelative, ComponentPose=@n45, bAlphaBoolEnabled:bool=True, ExternalForce="0,0,0") -> (Pose) {
    alphaInputType=Bool
    angularDampingOverride=0.9
    angularSpringConstant=11000
    bGravityOverrideInSimSpace=true
    bOverrideAngularDamping=true
    gravityScale=0.1
    linearSpringConstant=10
    simulationSpace=BoneRelative
    alphaBoolBlend {blendInTime=0.5, blendOption=QuadraticInOut, blendOutTime=1}
    alphaScaleBiasClamp {scale=0.11}
    boundBone {boneName=base_pieace_07}
    componentAppliedLinearAccClamp=<9500 9500 9500>
    componentLinearAccScale=<0.5 0.5 0.5>
    componentLinearVelScale=<0.5 0.5 0.5>
    gravityOverride {x=-1200}
    physicsBodyDefinitions [{
      sphereCollisionRadius=0
      boundBone {boneName=base_pieace_07}
      boxExtents=<35 42 40>
      constraintSetup {angularLimitsMax=<0 0 0>, angularLimitsMin=<0 0 0>, angularTarget=<0 0 0>, linearAxesMax=<11 1 1>, linearAxesMin=<-1 -1 -1>}
      localJointOffset=<2.5 0 0>
     }]
    relativeSpaceBone {boneName=Melee_TwoHanded_Handle}
    simSpaceSettings {maxAngularAcceleration=8, maxAngularVelocity=2.5, simSpaceAngularAlpha=1}
   }
   n47 = Anim Dynamics - Bone: base_pieace_08!(SimulationSpace:AnimPhysSimSpaceType Enum=BoneRelative, ComponentPose=@n46, bAlphaBoolEnabled:bool=True, ExternalForce="0,0,0") -> (Pose) {
    alphaInputType=Bool
    angularDampingOverride=0.9
    angularSpringConstant=11000
    bGravityOverrideInSimSpace=true
    bOverrideAngularDamping=true
    gravityScale=0.1
    linearSpringConstant=10
    simulationSpace=BoneRelative
    alphaBoolBlend {blendInTime=0.5, blendOption=QuadraticInOut, blendOutTime=1}
    alphaScaleBiasClamp {scale=0.11}
    boundBone {boneName=base_pieace_08}
    componentAppliedLinearAccClamp=<9500 9500 9500>
    componentLinearAccScale=<0.5 0.5 0.5>
    componentLinearVelScale=<0.5 0.5 0.5>
    gravityOverride {x=-1200}
    physicsBodyDefinitions [{
      sphereCollisionRadius=0
      boundBone {boneName=base_pieace_08}
      boxExtents=<40 37 39>
      constraintSetup {angularLimitsMax=<0 0 0>, angularLimitsMin=<0 0 0>, angularTarget=<0 0 0>, linearAxesMax=<12 1 1>, linearAxesMin=<-1 -1 -1>}
      localJointOffset=<3 0 0>
     }]
    relativeSpaceBone {boneName=Melee_TwoHanded_Handle}
    simSpaceSettings {maxAngularAcceleration=8, maxAngularVelocity=2.5, simSpaceAngularAlpha=1}
   }
   n48 = Anim Dynamics - Bone: base_pieace_09!(SimulationSpace:AnimPhysSimSpaceType Enum=BoneRelative, ComponentPose=@n47, bAlphaBoolEnabled:bool=True, ExternalForce="0,0,0") -> (Pose) {
    alphaInputType=Bool
    angularDampingOverride=0.9
    angularSpringConstant=11000
    bGravityOverrideInSimSpace=true
    bOverrideAngularDamping=true
    gravityScale=0.1
    linearSpringConstant=10
    simulationSpace=BoneRelative
    alphaBoolBlend {blendInTime=0.5, blendOption=QuadraticInOut, blendOutTime=1}
    alphaScaleBiasClamp {scale=0.11}
    boundBone {boneName=base_pieace_09}
    componentAppliedLinearAccClamp=<9500 9500 9500>
    componentLinearAccScale=<0.5 0.5 0.5>
    componentLinearVelScale=<0.5 0.5 0.5>
    gravityOverride {x=-1200}
    physicsBodyDefinitions [{
      sphereCollisionRadius=0
      boundBone {boneName=base_pieace_09}
      boxExtents=<40 38 40>
      constraintSetup {angularLimitsMax=<0 0 0>, angularLimitsMin=<0 0 0>, angularTarget=<0 0 0>, linearAxesMax=<14 1 1>, linearAxesMin=<-1 -1 -1>}
      localJointOffset=<2 0 0>
     }]
    relativeSpaceBone {boneName=Melee_TwoHanded_Handle}
    simSpaceSettings {maxAngularAcceleration=8, maxAngularVelocity=2.5, simSpaceAngularAlpha=1}
   }
   n49 = Anim Dynamics - Bone: base_pieace_10!(SimulationSpace:AnimPhysSimSpaceType Enum=BoneRelative, ComponentPose=@n48, bAlphaBoolEnabled:bool=True, ExternalForce="0,0,0") -> (Pose) {
    alphaInputType=Bool
    angularDampingOverride=0.9
    angularSpringConstant=11000
    bGravityOverrideInSimSpace=true
    bOverrideAngularDamping=true
    gravityScale=0.1
    linearSpringConstant=10
    simulationSpace=BoneRelative
    alphaBoolBlend {blendInTime=0.5, blendOption=QuadraticInOut, blendOutTime=1}
    alphaScaleBiasClamp {scale=0.11}
    boundBone {boneName=base_pieace_10}
    componentAppliedLinearAccClamp=<9500 9500 9500>
    componentLinearAccScale=<0.5 0.5 0.5>
    componentLinearVelScale=<0.5 0.5 0.5>
    gravityOverride {x=-1200}
    physicsBodyDefinitions [{
      sphereCollisionRadius=0
      boundBone {boneName=base_pieace_10}
      boxExtents=<36 45 40>
      constraintSetup {angularLimitsMax=<0 0 0>, angularLimitsMin=<0 0 0>, angularTarget=<0 0 0>, linearAxesMax=<13 1 1>, linearAxesMin=<-1 -1 -1>}
      localJointOffset=<2.5 0 0>
     }]
    relativeSpaceBone {boneName=Melee_TwoHanded_Handle}
    simSpaceSettings {maxAngularAcceleration=8, maxAngularVelocity=2.5, simSpaceAngularAlpha=1}
   }
   n50 = Anim Dynamics - Bone: base_pieace_11!(SimulationSpace:AnimPhysSimSpaceType Enum=BoneRelative, ComponentPose=@n49, bAlphaBoolEnabled:bool=True, ExternalForce="0,0,0") -> (Pose) {
    alphaInputType=Bool
    angularDampingOverride=0.9
    angularSpringConstant=11000
    bGravityOverrideInSimSpace=true
    bOverrideAngularDamping=true
    gravityScale=0.1
    linearSpringConstant=10
    simulationSpace=BoneRelative
    alphaBoolBlend {blendInTime=0.5, blendOption=QuadraticInOut, blendOutTime=1}
    alphaScaleBiasClamp {scale=0.11}
    boundBone {boneName=base_pieace_11}
    componentAppliedLinearAccClamp=<9500 9500 9500>
    componentLinearAccScale=<0.5 0.5 0.5>
    componentLinearVelScale=<0.5 0.5 0.5>
    gravityOverride {x=-1200}
    physicsBodyDefinitions [{
      sphereCollisionRadius=0
      boundBone {boneName=base_pieace_11}
      boxExtents=<55 50 55>
      constraintSetup {angularLimitsMax=<0 0 0>, angularLimitsMin=<0 0 0>, angularTarget=<0 0 0>, linearAxesMax=<11 1 1>, linearAxesMin=<-1 -1 -1>}
      localJointOffset=<2 0 0>
     }]
    relativeSpaceBone {boneName=Melee_TwoHanded_Handle}
    simSpaceSettings {maxAngularAcceleration=8, maxAngularVelocity=2.5, simSpaceAngularAlpha=1}
   }
   n51 = Anim Dynamics - Bone: base_pieace_12!(SimulationSpace:AnimPhysSimSpaceType Enum=BoneRelative, ComponentPose=@n50, bAlphaBoolEnabled:bool=True, ExternalForce="0,0,0") -> (Pose) {
    alphaInputType=Bool
    angularDampingOverride=0.9
    angularSpringConstant=11000
    bGravityOverrideInSimSpace=true
    bOverrideAngularDamping=true
    gravityScale=0.1
    linearSpringConstant=10
    simulationSpace=BoneRelative
    alphaBoolBlend {blendInTime=0.5, blendOption=QuadraticInOut, blendOutTime=1}
    alphaScaleBiasClamp {scale=0.11}
    boundBone {boneName=base_pieace_12}
    componentAppliedLinearAccClamp=<9500 9500 9500>
    componentLinearAccScale=<0.5 0.5 0.5>
    componentLinearVelScale=<0.5 0.5 0.5>
    gravityOverride {x=-1200}
    physicsBodyDefinitions [{
      sphereCollisionRadius=0
      boundBone {boneName=base_pieace_12}
      boxExtents=<50 55 40>
      constraintSetup {angularLimitsMax=<0 0 0>, angularLimitsMin=<0 0 0>, angularTarget=<0 0 0>, linearAxesMax=<12 1 1>, linearAxesMin=<-1 -1 -1>}
      localJointOffset=<1.5 0 0>
     }]
    relativeSpaceBone {boneName=Melee_TwoHanded_Handle}
    simSpaceSettings {maxAngularAcceleration=8, maxAngularVelocity=2.5, simSpaceAngularAlpha=1}
   }
   n52 = Anim Dynamics - Bone: base_pieace_13!(SimulationSpace:AnimPhysSimSpaceType Enum=BoneRelative, ComponentPose=@n51, bAlphaBoolEnabled:bool=True, ExternalForce="0,0,0") -> (Pose) {
    alphaInputType=Bool
    angularDampingOverride=0.9
    angularSpringConstant=11000
    bGravityOverrideInSimSpace=true
    bOverrideAngularDamping=true
    gravityScale=0.1
    linearSpringConstant=10
    simulationSpace=BoneRelative
    alphaBoolBlend {blendInTime=0.5, blendOption=QuadraticInOut, blendOutTime=1}
    alphaScaleBiasClamp {scale=0.11}
    boundBone {boneName=base_pieace_13}
    componentAppliedLinearAccClamp=<9500 9500 9500>
    componentLinearAccScale=<0.5 0.5 0.5>
    componentLinearVelScale=<0.5 0.5 0.5>
    gravityOverride {x=-1200}
    physicsBodyDefinitions [{
      sphereCollisionRadius=0
      boundBone {boneName=base_pieace_13}
      boxExtents=<60 50 65>
      constraintSetup {angularLimitsMax=<0 0 0>, angularLimitsMin=<0 0 0>, angularTarget=<0 0 0>, linearAxesMax=<14 1 1>, linearAxesMin=<-1 -1 -1>}
      localJointOffset=<1 0 0>
     }]
    relativeSpaceBone {boneName=Melee_TwoHanded_Handle}
    simSpaceSettings {maxAngularAcceleration=8, maxAngularVelocity=2.5, simSpaceAngularAlpha=1}
   }
   n53 = Anim Dynamics - Bone: base_pieace_14!(SimulationSpace:AnimPhysSimSpaceType Enum=BoneRelative, ComponentPose=@n52, bAlphaBoolEnabled:bool=True, ExternalForce="0,0,0") -> (Pose) {
    alphaInputType=Bool
    angularDampingOverride=0.9
    angularSpringConstant=11000
    bGravityOverrideInSimSpace=true
    bOverrideAngularDamping=true
    gravityScale=0.1
    linearSpringConstant=10
    simulationSpace=BoneRelative
    alphaBoolBlend {blendInTime=0.5, blendOption=QuadraticInOut, blendOutTime=1}
    alphaScaleBiasClamp {scale=0.11}
    boundBone {boneName=base_pieace_14}
    componentAppliedLinearAccClamp=<9500 9500 9500>
    componentLinearAccScale=<0.5 0.5 0.5>
    componentLinearVelScale=<0.5 0.5 0.5>
    gravityOverride {x=-1200}
    physicsBodyDefinitions [{
      sphereCollisionRadius=0
      boundBone {boneName=base_pieace_14}
      boxExtents=<42 41 40>
      constraintSetup {angularLimitsMax=<0 0 0>, angularLimitsMin=<0 0 0>, angularTarget=<0 0 0>, linearAxesMax=<11 1 1>, linearAxesMin=<-1 -1 -1>}
      localJointOffset=<1.8 0 0>
     }]
    relativeSpaceBone {boneName=Melee_TwoHanded_Handle}
    simSpaceSettings {maxAngularAcceleration=8, maxAngularVelocity=2.5, simSpaceAngularAlpha=1}
   }
   n54 = Anim Dynamics - Bone: base_pieace_15!(SimulationSpace:AnimPhysSimSpaceType Enum=BoneRelative, ComponentPose=@n53, bAlphaBoolEnabled:bool=True, ExternalForce="0,0,0") -> (Pose) {
    alphaInputType=Bool
    angularDampingOverride=0.9
    angularSpringConstant=11000
    bGravityOverrideInSimSpace=true
    bOverrideAngularDamping=true
    gravityScale=0.1
    linearSpringConstant=10
    simulationSpace=BoneRelative
    alphaBoolBlend {blendInTime=0.5, blendOption=QuadraticInOut, blendOutTime=1}
    alphaScaleBiasClamp {scale=0.11}
    boundBone {boneName=base_pieace_15}
    componentAppliedLinearAccClamp=<9500 9500 9500>
    componentLinearAccScale=<0.5 0.5 0.5>
    componentLinearVelScale=<0.5 0.5 0.5>
    gravityOverride {x=-1200}
    physicsBodyDefinitions [{
      sphereCollisionRadius=0
      boundBone {boneName=base_pieace_15}
      boxExtents=<40 40 40>
      constraintSetup {angularLimitsMax=<0 0 0>, angularLimitsMin=<0 0 0>, angularTarget=<0 0 0>, linearAxesMax=<10 1 1>, linearAxesMin=<-1 -1 -1>}
      localJointOffset=<2 0 0>
     }]
    relativeSpaceBone {boneName=Melee_TwoHanded_Handle}
    simSpaceSettings {maxAngularAcceleration=8, maxAngularVelocity=2.5, simSpaceAngularAlpha=1}
   }
   n55 = Anim Dynamics - Bone: base_pieace_16!(SimulationSpace:AnimPhysSimSpaceType Enum=BoneRelative, ComponentPose=@n54, bAlphaBoolEnabled:bool=True, ExternalForce="0,0,0") -> (Pose) {
    alphaInputType=Bool
    angularDampingOverride=0.9
    angularSpringConstant=11000
    bGravityOverrideInSimSpace=true
    bOverrideAngularDamping=true
    gravityScale=0.1
    linearSpringConstant=10
    simulationSpace=BoneRelative
    alphaBoolBlend {blendInTime=0.5, blendOption=QuadraticInOut, blendOutTime=1}
    alphaScaleBiasClamp {scale=0.11}
    boundBone {boneName=base_pieace_16}
    componentAppliedLinearAccClamp=<9500 9500 9500>
    componentLinearAccScale=<0.5 0.5 0.5>
    componentLinearVelScale=<0.5 0.5 0.5>
    gravityOverride {x=-1200}
    physicsBodyDefinitions [{
      sphereCollisionRadius=0
      boundBone {boneName=base_pieace_16}
      boxExtents=<30 30 30>
      constraintSetup {angularLimitsMax=<0 0 0>, angularLimitsMin=<0 0 0>, angularTarget=<0 0 0>, linearAxesMax=<13 1 1>, linearAxesMin=<-1 -1 -1>}
      localJointOffset=<2 0 0>
     }]
    relativeSpaceBone {boneName=Melee_TwoHanded_Handle}
    simSpaceSettings {maxAngularAcceleration=8, maxAngularVelocity=2.5, simSpaceAngularAlpha=1}
   }
   n56 = Anim Dynamics - Bone: base_pieace_17!(SimulationSpace:AnimPhysSimSpaceType Enum=BoneRelative, ComponentPose=@n55, bAlphaBoolEnabled:bool=True, ExternalForce="0,0,0") -> (Pose) {
    alphaInputType=Bool
    angularDampingOverride=0.9
    angularSpringConstant=11000
    bGravityOverrideInSimSpace=true
    bOverrideAngularDamping=true
    gravityScale=0.1
    linearSpringConstant=10
    simulationSpace=BoneRelative
    alphaBoolBlend {blendInTime=0.5, blendOption=QuadraticInOut, blendOutTime=1}
    alphaScaleBiasClamp {scale=0.11}
    boundBone {boneName=base_pieace_17}
    componentAppliedLinearAccClamp=<9500 9500 9500>
    componentLinearAccScale=<0.5 0.5 0.5>
    componentLinearVelScale=<0.5 0.5 0.5>
    gravityOverride {x=-1200}
    physicsBodyDefinitions [{
      sphereCollisionRadius=0
      boundBone {boneName=base_pieace_17}
      boxExtents=<35 30 38>
      constraintSetup {angularLimitsMax=<0 0 0>, angularLimitsMin=<0 0 0>, angularTarget=<0 0 0>, linearAxesMax=<14 1 1>, linearAxesMin=<-1 -1 -1>}
      localJointOffset=<3 0 0>
     }]
    relativeSpaceBone {boneName=Melee_TwoHanded_Handle}
    simSpaceSettings {maxAngularAcceleration=8, maxAngularVelocity=2.5, simSpaceAngularAlpha=1}
   }
   n57 = Anim Dynamics - Bone: base_pieace_18!(SimulationSpace:AnimPhysSimSpaceType Enum=BoneRelative, ComponentPose=@n56, bAlphaBoolEnabled:bool=True, ExternalForce="0,0,0") -> (Pose) {
    alphaInputType=Bool
    angularDampingOverride=0.9
    angularSpringConstant=11000
    bGravityOverrideInSimSpace=true
    bOverrideAngularDamping=true
    gravityScale=0.1
    linearSpringConstant=10
    simulationSpace=BoneRelative
    alphaBoolBlend {blendInTime=0.5, blendOption=QuadraticInOut, blendOutTime=1}
    alphaScaleBiasClamp {scale=0.11}
    boundBone {boneName=base_pieace_18}
    componentAppliedLinearAccClamp=<9500 9500 9500>
    componentLinearAccScale=<0.5 0.5 0.5>
    componentLinearVelScale=<0.5 0.5 0.5>
    gravityOverride {x=-1200}
    physicsBodyDefinitions [{
      sphereCollisionRadius=0
      boundBone {boneName=base_pieace_18}
      boxExtents=<40 40 40>
      constraintSetup {angularLimitsMax=<0 0 0>, angularLimitsMin=<0 0 0>, angularTarget=<0 0 0>, linearAxesMax=<9 1 1>, linearAxesMin=<-1 -1 -1>}
      localJointOffset=<2 0 0>
     }]
    relativeSpaceBone {boneName=Melee_TwoHanded_Handle}
    simSpaceSettings {maxAngularAcceleration=8, maxAngularVelocity=2.5, simSpaceAngularAlpha=1}
   }
   n58 = Anim Dynamics - Bone: base_pieace_19!(SimulationSpace:AnimPhysSimSpaceType Enum=BoneRelative, ComponentPose=@n57, bAlphaBoolEnabled:bool=True, ExternalForce="0,0,0") -> (Pose) {
    alphaInputType=Bool
    angularDampingOverride=0.9
    angularSpringConstant=11000
    bGravityOverrideInSimSpace=true
    bOverrideAngularDamping=true
    gravityScale=0.1
    linearSpringConstant=10
    simulationSpace=BoneRelative
    alphaBoolBlend {blendInTime=0.5, blendOption=QuadraticInOut, blendOutTime=1}
    alphaScaleBiasClamp {scale=0.11}
    boundBone {boneName=base_pieace_19}
    componentAppliedLinearAccClamp=<9500 9500 9500>
    componentLinearAccScale=<0.5 0.5 0.5>
    componentLinearVelScale=<0.5 0.5 0.5>
    gravityOverride {x=-1200}
    physicsBodyDefinitions [{
      sphereCollisionRadius=0
      boundBone {boneName=base_pieace_19}
      boxExtents=<38 40 39>
      constraintSetup {angularLimitsMax=<0 0 0>, angularLimitsMin=<0 0 0>, angularTarget=<0 0 0>, linearAxesMax=<1 1 1>, linearAxesMin=<-1 -13 -1>}
      localJointOffset=<2 0 0>
     }]
    relativeSpaceBone {boneName=Melee_TwoHanded_Handle}
    simSpaceSettings {maxAngularAcceleration=8, maxAngularVelocity=2.5, simSpaceAngularAlpha=1}
   }
   n59 = Anim Dynamics - Bone: base_pieace_20!(SimulationSpace:AnimPhysSimSpaceType Enum=BoneRelative, ComponentPose=@n58, bAlphaBoolEnabled:bool=True, ExternalForce="0,0,0") -> (Pose) {
    alphaInputType=Bool
    angularDampingOverride=0.9
    angularSpringConstant=11000
    bGravityOverrideInSimSpace=true
    bOverrideAngularDamping=true
    gravityScale=0.1
    linearSpringConstant=10
    simulationSpace=BoneRelative
    alphaBoolBlend {blendInTime=0.5, blendOption=QuadraticInOut, blendOutTime=1}
    alphaScaleBiasClamp {scale=0.11}
    boundBone {boneName=base_pieace_20}
    componentAppliedLinearAccClamp=<9500 9500 9500>
    componentLinearAccScale=<0.5 0.5 0.5>
    componentLinearVelScale=<0.5 0.5 0.5>
    gravityOverride {x=-1200}
    physicsBodyDefinitions [{
      sphereCollisionRadius=0
      boundBone {boneName=base_pieace_20}
      boxExtents=<35 35 35>
      constraintSetup {angularLimitsMax=<0 0 0>, angularLimitsMin=<0 0 0>, angularTarget=<0 0 0>, linearAxesMax=<14 1 1>, linearAxesMin=<-1 -1 -1>}
      localJointOffset=<3 0 0>
     }]
    relativeSpaceBone {boneName=Melee_TwoHanded_Handle}
    simSpaceSettings {maxAngularAcceleration=8, maxAngularVelocity=2.5, simSpaceAngularAlpha=1}
   }
   Output Pose(Result=Component To Local(ComponentPose=@n59)) {layerGroup=None}
   AnimGraphNode_SubInput_0 = Input Pose()
   n60 = "Transform (Modify) Bone - Bone: dyn_pieace_01"(ComponentPose=Local To Component(LocalPose=@AnimGraphNode_SubInput_0), Alpha=0.5, Translation="0,0,0", Rotation="0,0,0", Scale="1,1,1") {
    alpha=0.5
    rotationMode=BMM_Additive
    translationMode=BMM_Additive
    translationSpace=BCS_BoneSpace
    alphaBoolBlend {blendOption=Linear}
    boneToModify {boneName=dyn_pieace_01}
   }
   n61 = "Transform (Modify) Bone - Bone: dyn_pieace_02"(ComponentPose=@n60, Alpha=1.1, Translation="0,0,0", Rotation="0,0,0", Scale="1,1,1") {
    rotationMode=BMM_Additive
    translationMode=BMM_Additive
    translationSpace=BCS_BoneSpace
    alphaBoolBlend {blendOption=Linear}
    boneToModify {boneName=dyn_pieace_02}
   }
   n62 = "Transform (Modify) Bone - Bone: dyn_pieace_03"(ComponentPose=@n61, Alpha=1, Translation="0,0,0", Rotation="0,0,0", Scale="1,1,1") {
    rotationMode=BMM_Additive
    translationMode=BMM_Additive
    translationSpace=BCS_BoneSpace
    alphaBoolBlend {blendOption=Linear}
    boneToModify {boneName=dyn_pieace_03}
   }
   n63 = "Transform (Modify) Bone - Bone: dyn_pieace_04"(ComponentPose=@n62, Alpha=0.7, Translation="0,0,0", Rotation="0,0,0", Scale="1,1,1") {
    rotationMode=BMM_Additive
    translationMode=BMM_Additive
    translationSpace=BCS_BoneSpace
    alphaBoolBlend {blendOption=Linear}
    boneToModify {boneName=dyn_pieace_04}
   }
   n64 = "Transform (Modify) Bone - Bone: base_pieace_05"(ComponentPose=@n63, Alpha=0.8, Translation="0,0,0", Rotation="0,0,0", Scale="1,1,1") {
    rotationMode=BMM_Additive
    translationMode=BMM_Additive
    translationSpace=BCS_BoneSpace
    alphaBoolBlend {blendOption=Linear}
    boneToModify {boneName=base_pieace_05}
   }
   n65 = "Transform (Modify) Bone - Bone: dyn_pieace_06"(ComponentPose=@n64, Alpha=0.2, Translation="0,0,0", Rotation="0,0,0", Scale="1,1,1") {
    rotationMode=BMM_Additive
    translationMode=BMM_Additive
    translationSpace=BCS_BoneSpace
    alphaBoolBlend {blendOption=Linear}
    boneToModify {boneName=dyn_pieace_06}
   }
   n66 = "Transform (Modify) Bone - Bone: dyn_pieace_07"(ComponentPose=@n65, Alpha=0.7, Translation="0,0,0", Rotation="0,0,0", Scale="1,1,1") {
    rotationMode=BMM_Additive
    translationMode=BMM_Additive
    translationSpace=BCS_BoneSpace
    alphaBoolBlend {blendOption=Linear}
    boneToModify {boneName=dyn_pieace_07}
   }
   n67 = "Transform (Modify) Bone - Bone: dyn_pieace_08"(ComponentPose=@n66, Alpha=0.7, Translation="0,0,0", Rotation="0,0,0", Scale="1,1,1") {
    rotationMode=BMM_Additive
    translationMode=BMM_Additive
    translationSpace=BCS_BoneSpace
    alphaBoolBlend {blendOption=Linear}
    boneToModify {boneName=dyn_pieace_08}
   }
   n68 = "Transform (Modify) Bone - Bone: dyn_pieace_09"(ComponentPose=@n67, Alpha=1.6, Translation="0,0,0", Rotation="0,0,0", Scale="1,1,1") {
    rotationMode=BMM_Additive
    translationMode=BMM_Additive
    translationSpace=BCS_BoneSpace
    alphaBoolBlend {blendOption=Linear}
    boneToModify {boneName=dyn_pieace_09}
   }
   n69 = "Transform (Modify) Bone - Bone: dyn_pieace_10"(ComponentPose=@n68, Alpha=0.7, Translation="0,0,0", Rotation="0,0,0", Scale="1,1,1") {
    rotationMode=BMM_Additive
    translationMode=BMM_Additive
    translationSpace=BCS_BoneSpace
    alphaBoolBlend {blendInTime=0.5, blendOption=CircularIn, blendOutTime=0.5}
    boneToModify {boneName=dyn_pieace_10}
   }
   n70 = "Transform (Modify) Bone - Bone: dyn_pieace_11"(ComponentPose=@n69, Alpha=0.5, Translation="0,0,0", Rotation="0,0,0", Scale="1,1,1") {
    alpha=0.5
    rotationMode=BMM_Additive
    translationMode=BMM_Additive
    translationSpace=BCS_BoneSpace
    alphaBoolBlend {blendOption=Linear}
    boneToModify {boneName=dyn_pieace_11}
   }
   n71 = "Transform (Modify) Bone - Bone: dyn_pieace_12"(ComponentPose=@n70, Alpha=1.1, Translation="0,0,0", Rotation="0,0,0", Scale="1,1,1") {
    rotationMode=BMM_Additive
    translationMode=BMM_Additive
    translationSpace=BCS_BoneSpace
    alphaBoolBlend {blendOption=Linear}
    boneToModify {boneName=dyn_pieace_12}
   }
   n72 = "Transform (Modify) Bone - Bone: dyn_pieace_13"(ComponentPose=@n71, Alpha=1, Translation="0,0,0", Rotation="0,0,0", Scale="1,1,1") {
    rotationMode=BMM_Additive
    translationMode=BMM_Additive
    translationSpace=BCS_BoneSpace
    alphaBoolBlend {blendOption=Linear}
    boneToModify {boneName=dyn_pieace_13}
   }
   n73 = "Transform (Modify) Bone - Bone: dyn_pieace_14"(ComponentPose=@n72, Alpha=0.7, Translation="0,0,0", Rotation="0,0,0", Scale="1,1,1") {
    rotationMode=BMM_Additive
    translationMode=BMM_Additive
    translationSpace=BCS_BoneSpace
    alphaBoolBlend {blendOption=Linear}
    boneToModify {boneName=dyn_pieace_14}
   }
   n74 = "Transform (Modify) Bone - Bone: dyn_pieace_15"(ComponentPose=@n73, Alpha=0.8, Translation="0,0,0", Rotation="0,0,0", Scale="1,1,1") {
    rotationMode=BMM_Additive
    translationMode=BMM_Additive
    translationSpace=BCS_BoneSpace
    alphaBoolBlend {blendOption=Linear}
    boneToModify {boneName=dyn_pieace_15}
   }
   n75 = "Transform (Modify) Bone - Bone: dyn_pieace_16"(ComponentPose=@n74, Alpha=1.6, Translation="0,0,0", Rotation="0,0,0", Scale="1,1,1") {
    rotationMode=BMM_Additive
    translationMode=BMM_Additive
    translationSpace=BCS_BoneSpace
    alphaBoolBlend {blendOption=Linear}
    boneToModify {boneName=dyn_pieace_16}
   }
   n76 = "Transform (Modify) Bone - Bone: dyn_pieace_17"(ComponentPose=@n75, Alpha=0.5, Translation="0,0,0", Rotation="0,0,0", Scale="1,1,1") {
    rotationMode=BMM_Additive
    translationMode=BMM_Additive
    translationSpace=BCS_BoneSpace
    alphaBoolBlend {blendOption=Linear}
    boneToModify {boneName=dyn_pieace_17}
   }
   n77 = "Transform (Modify) Bone - Bone: dyn_pieace_18"(ComponentPose=@n76, Alpha=0.7, Translation="0,0,0", Rotation="0,0,0", Scale="1,1,1") {
    rotationMode=BMM_Additive
    translationMode=BMM_Additive
    translationSpace=BCS_BoneSpace
    alphaBoolBlend {blendOption=Linear}
    boneToModify {boneName=dyn_pieace_18}
   }
   n78 = "Transform (Modify) Bone - Bone: dyn_pieace_19"(ComponentPose=@n77, Alpha=0.2, Translation="0,0,0", Rotation="0,0,0", Scale="1,1,1") {
    rotationMode=BMM_Additive
    translationMode=BMM_Additive
    translationSpace=BCS_BoneSpace
    alphaBoolBlend {blendOption=Linear}
    boneToModify {boneName=dyn_pieace_19}
   }
   n79 = "Transform (Modify) Bone - Bone: dyn_pieace_20"(ComponentPose=@n78, Alpha=0.7, Translation="0,0,0", Rotation="0,0,0", Scale="1,1,1") {
    rotationMode=BMM_Additive
    translationMode=BMM_Additive
    translationSpace=BCS_BoneSpace
    alphaBoolBlend {blendInTime=0.5, blendOption=CircularIn, blendOutTime=0.5}
    boneToModify {boneName=dyn_pieace_20}
   }
   n80 = "Transform (Modify) Bone - Bone: base_pieace_04"!(ComponentPose=@n79, bAlphaBoolEnabled:bool=True, Rotation="0,0,0", Scale="1,1,1") -> (Pose) {
    alphaInputType=Bool
    translationMode=BMM_Additive
    translationSpace=BCS_BoneSpace
    alphaBoolBlend {blendInTime=0.2, blendOption=Linear, blendOutTime=1}
    boneToModify {boneName=base_pieace_04}
    translation {x=15}
   }
   n81 = "Transform (Modify) Bone - Bone: base_pieace_20"!(ComponentPose=@n80, bAlphaBoolEnabled:bool=True, Rotation="0,0,0", Scale="1,1,1") -> (Pose) {
    alphaInputType=Bool
    translationMode=BMM_Additive
    translationSpace=BCS_BoneSpace
    alphaBoolBlend {blendInTime=0.2, blendOption=Linear, blendOutTime=1}
    boneToModify {boneName=base_pieace_20}
    translation {x=15}
   }
   n82 = "Transform (Modify) Bone - Bone: base_pieace_12"!(ComponentPose=@n81, bAlphaBoolEnabled:bool=True, Rotation="0,0,0", Scale="1,1,1") -> (Pose) {
    alphaInputType=Bool
    translationMode=BMM_Additive
    translationSpace=BCS_BoneSpace
    alphaBoolBlend {blendInTime=0.2, blendOption=Linear, blendOutTime=1}
    boneToModify {boneName=base_pieace_12}
    translation {x=15}
   }
   n83 = "Transform (Modify) Bone - Bone: base_pieace_07"!(ComponentPose=@n82, bAlphaBoolEnabled:bool=True, Rotation="0,0,0", Scale="1,1,1") -> (Pose) {
    alphaInputType=Bool
    translationMode=BMM_Additive
    translationSpace=BCS_BoneSpace
    alphaBoolBlend {blendInTime=0.2, blendOption=Linear, blendOutTime=1}
    boneToModify {boneName=base_pieace_07}
    translation {x=15}
   }
   n84 = "Transform (Modify) Bone - Bone: base_pieace_11"!(ComponentPose=@n83, bAlphaBoolEnabled:bool=True, Rotation="0,0,0", Scale="1,1,1") -> (Pose) {
    alphaInputType=Bool
    translationMode=BMM_Additive
    translationSpace=BCS_BoneSpace
    alphaBoolBlend {blendInTime=0.2, blendOption=Linear, blendOutTime=1}
    boneToModify {boneName=base_pieace_11}
    translation {x=15}
   }
   n85 = "Transform (Modify) Bone - Bone: base_pieace_03"!(ComponentPose=@n84, bAlphaBoolEnabled:bool=True, Rotation="0,0,0", Scale="1,1,1") -> (Pose) {
    alphaInputType=Bool
    translationMode=BMM_Additive
    translationSpace=BCS_BoneSpace
    alphaBoolBlend {blendInTime=0.2, blendOption=Linear, blendOutTime=1}
    boneToModify {boneName=base_pieace_03}
    translation {x=15}
   }
   n86 = "Transform (Modify) Bone - Bone: base_pieace_13"!(ComponentPose=@n85, bAlphaBoolEnabled:bool=True, Rotation="0,0,0", Scale="1,1,1") -> (Pose) {
    alphaInputType=Bool
    translationMode=BMM_Additive
    translationSpace=BCS_BoneSpace
    alphaBoolBlend {blendInTime=0.2, blendOption=Linear, blendOutTime=1}
    boneToModify {boneName=base_pieace_13}
    translation {x=12}
   }
   n87 = "Transform (Modify) Bone - Bone: base_pieace_08"!(ComponentPose=@n86, bAlphaBoolEnabled:bool=True, Rotation="0,0,0", Scale="1,1,1") -> (Pose) {
    alphaInputType=Bool
    translationMode=BMM_Additive
    translationSpace=BCS_BoneSpace
    alphaBoolBlend {blendInTime=0.2, blendOption=Linear, blendOutTime=1}
    boneToModify {boneName=base_pieace_08}
    translation {x=10}
   }
   n88 = "Transform (Modify) Bone - Bone: base_pieace_10"!(ComponentPose=@n87, bAlphaBoolEnabled:bool=True, Rotation="0,0,0", Scale="1,1,1") -> (Pose) {
    alphaInputType=Bool
    translationMode=BMM_Additive
    translationSpace=BCS_BoneSpace
    alphaBoolBlend {blendInTime=0.2, blendOption=Linear, blendOutTime=1}
    boneToModify {boneName=base_pieace_10}
    translation {x=15}
   }
   n89 = "Transform (Modify) Bone - Bone: base_pieace_19"!(ComponentPose=@n88, bAlphaBoolEnabled:bool=True, Translation="-1.000,-8.000,0.000", Rotation="0,0,0", Scale="1,1,1") -> (Pose) {
    alphaInputType=Bool
    translationMode=BMM_Additive
    translationSpace=BCS_BoneSpace
    alphaBoolBlend {blendInTime=0.2, blendOption=Linear, blendOutTime=1}
    boneToModify {boneName=base_pieace_19}
    translation=<-1 -8>
   }
   n90 = "Transform (Modify) Bone - Bone: base_pieace_18"!(ComponentPose=@n89, bAlphaBoolEnabled:bool=True, Rotation="0,0,0", Scale="1,1,1") -> (Pose) {
    alphaInputType=Bool
    translationMode=BMM_Additive
    translationSpace=BCS_BoneSpace
    alphaBoolBlend {blendInTime=0.2, blendOption=Linear, blendOutTime=1}
    boneToModify {boneName=base_pieace_18}
    translation {x=17}
   }
   n91 = "Transform (Modify) Bone - Bone: base_pieace_09"!(ComponentPose=@n90, bAlphaBoolEnabled:bool=True, Rotation="0,0,0", Scale="1,1,1") -> (Pose) {
    alphaInputType=Bool
    translationMode=BMM_Additive
    translationSpace=BCS_BoneSpace
    alphaBoolBlend {blendInTime=0.2, blendOption=Linear, blendOutTime=1}
    boneToModify {boneName=base_pieace_09}
    translation {x=10}
   }
   n92 = "Transform (Modify) Bone - Bone: base_pieace_06"!(ComponentPose=@n91, bAlphaBoolEnabled:bool=True, Rotation="0,0,0", Scale="1,1,1") -> (Pose) {
    alphaInputType=Bool
    translationMode=BMM_Additive
    translationSpace=BCS_BoneSpace
    alphaBoolBlend {blendInTime=0.2, blendOption=Linear, blendOutTime=1}
    boneToModify {boneName=base_pieace_06}
    translation {x=17}
   }
   n93 = "Transform (Modify) Bone - Bone: base_pieace_02"!(ComponentPose=@n92, bAlphaBoolEnabled:bool=True, Rotation="0,0,0", Scale="1,1,1") -> (Pose) {
    alphaInputType=Bool
    translationMode=BMM_Additive
    translationSpace=BCS_BoneSpace
    alphaBoolBlend {blendInTime=0.2, blendOption=Linear, blendOutTime=1}
    boneToModify {boneName=base_pieace_02}
    translation {x=17}
   }
   n94 = "Transform (Modify) Bone - Bone: base_pieace_16"!(ComponentPose=@n93, bAlphaBoolEnabled:bool=True, Rotation="0,0,0", Scale="1,1,1") -> (Pose) {
    alphaInputType=Bool
    translationMode=BMM_Additive
    translationSpace=BCS_BoneSpace
    alphaBoolBlend {blendInTime=0.2, blendOption=Linear, blendOutTime=1}
    boneToModify {boneName=base_pieace_16}
    translation {x=17}
   }
   n95 = "Transform (Modify) Bone - Bone: base_pieace_17"!(ComponentPose=@n94, bAlphaBoolEnabled:bool=True, Rotation="0,0,0", Scale="1,1,1") -> (Pose) {
    alphaInputType=Bool
    translationMode=BMM_Additive
    translationSpace=BCS_BoneSpace
    alphaBoolBlend {blendInTime=0.2, blendOption=Linear, blendOutTime=1}
    boneToModify {boneName=base_pieace_17}
    translation {x=17}
   }
   n96 = "Transform (Modify) Bone - Bone: dyn_pieace_01"(ComponentPose=@n95, Alpha=1, Translation="0,0,0", Rotation="0,0,0", Scale="1,1,1") {
    alpha=0.5
    rotationMode=BMM_Additive
    translationMode=BMM_Additive
    translationSpace=BCS_BoneSpace
    alphaBoolBlend {blendOption=Linear}
    boneToModify {boneName=dyn_pieace_01}
   }
   n97 = "Transform (Modify) Bone - Bone: dyn_pieace_02"(ComponentPose=@n96, Alpha=1, Translation="0,0,0", Rotation="0,0,0", Scale="1,1,1") {
    rotationMode=BMM_Additive
    translationMode=BMM_Additive
    translationSpace=BCS_BoneSpace
    alphaBoolBlend {blendOption=Linear}
    boneToModify {boneName=dyn_pieace_02}
   }
   n98 = "Transform (Modify) Bone - Bone: dyn_pieace_03"(ComponentPose=@n97, Alpha=1, Translation="0,0,0", Rotation="0,0,0", Scale="1,1,1") {
    rotationMode=BMM_Additive
    translationMode=BMM_Additive
    translationSpace=BCS_BoneSpace
    alphaBoolBlend {blendOption=Linear}
    boneToModify {boneName=dyn_pieace_03}
   }
   n99 = "Transform (Modify) Bone - Bone: dyn_pieace_04"(ComponentPose=@n98, Alpha=1, Translation="0,0,0", Rotation="0,0,0", Scale="1,1,1") {
    rotationMode=BMM_Additive
    translationMode=BMM_Additive
    translationSpace=BCS_BoneSpace
    alphaBoolBlend {blendOption=Linear}
    boneToModify {boneName=dyn_pieace_04}
   }
   n100 = "Transform (Modify) Bone - Bone: base_pieace_05"(ComponentPose=@n99, Alpha=1, Translation="0,0,0", Rotation="0,0,0", Scale="1,1,1") {
    rotationMode=BMM_Additive
    translationMode=BMM_Additive
    translationSpace=BCS_BoneSpace
    alphaBoolBlend {blendOption=Linear}
    boneToModify {boneName=base_pieace_05}
   }
   n101 = "Transform (Modify) Bone - Bone: dyn_pieace_06"(ComponentPose=@n100, Alpha=1, Translation="0,0,0", Rotation="0,0,0", Scale="1,1,1") {
    rotationMode=BMM_Additive
    translationMode=BMM_Additive
    translationSpace=BCS_BoneSpace
    alphaBoolBlend {blendOption=Linear}
    boneToModify {boneName=dyn_pieace_06}
   }
   n102 = "Transform (Modify) Bone - Bone: dyn_pieace_07"(ComponentPose=@n101, Alpha=1, Translation="0,0,0", Rotation="0,0,0", Scale="1,1,1") {
    rotationMode=BMM_Additive
    translationMode=BMM_Additive
    translationSpace=BCS_BoneSpace
    alphaBoolBlend {blendOption=Linear}
    boneToModify {boneName=dyn_pieace_07}
   }
   n103 = "Transform (Modify) Bone - Bone: dyn_pieace_08"(ComponentPose=@n102, Alpha=1, Translation="0,0,0", Rotation="0,0,0", Scale="1,1,1") {
    rotationMode=BMM_Additive
    translationMode=BMM_Additive
    translationSpace=BCS_BoneSpace
    alphaBoolBlend {blendOption=Linear}
    boneToModify {boneName=dyn_pieace_08}
   }
   n104 = "Transform (Modify) Bone - Bone: dyn_pieace_09"(ComponentPose=@n103, Alpha=1, Translation="0,0,0", Rotation="0,0,0", Scale="1,1,1") {
    rotationMode=BMM_Additive
    translationMode=BMM_Additive
    translationSpace=BCS_BoneSpace
    alphaBoolBlend {blendOption=Linear}
    boneToModify {boneName=dyn_pieace_09}
   }
   n105 = "Transform (Modify) Bone - Bone: dyn_pieace_10"(ComponentPose=@n104, Alpha=1, Translation="0,0,0", Rotation="0,0,0", Scale="1,1,1") {
    rotationMode=BMM_Additive
    translationMode=BMM_Additive
    translationSpace=BCS_BoneSpace
    alphaBoolBlend {blendInTime=0.5, blendOption=CircularIn, blendOutTime=0.5}
    boneToModify {boneName=dyn_pieace_10}
   }
   n106 = "Transform (Modify) Bone - Bone: dyn_pieace_11"(ComponentPose=@n105, Alpha=1, Translation="0,0,0", Rotation="0,0,0", Scale="1,1,1") {
    alpha=0.5
    rotationMode=BMM_Additive
    translationMode=BMM_Additive
    translationSpace=BCS_BoneSpace
    alphaBoolBlend {blendOption=Linear}
    boneToModify {boneName=dyn_pieace_11}
   }
   n107 = "Transform (Modify) Bone - Bone: dyn_pieace_12"(ComponentPose=@n106, Alpha=1.1, Translation="0,0,0", Rotation="0,0,0", Scale="1,1,1") {
    rotationMode=BMM_Additive
    translationMode=BMM_Additive
    translationSpace=BCS_BoneSpace
    alphaBoolBlend {blendOption=Linear}
    boneToModify {boneName=dyn_pieace_12}
   }
   n108 = "Transform (Modify) Bone - Bone: dyn_pieace_13"(ComponentPose=@n107, Alpha=1, Translation="0,0,0", Rotation="0,0,0", Scale="1,1,1") {
    rotationMode=BMM_Additive
    translationMode=BMM_Additive
    translationSpace=BCS_BoneSpace
    alphaBoolBlend {blendOption=Linear}
    boneToModify {boneName=dyn_pieace_13}
   }
   n109 = "Transform (Modify) Bone - Bone: dyn_pieace_14"(ComponentPose=@n108, Alpha=1, Translation="0,0,0", Rotation="0,0,0", Scale="1,1,1") {
    rotationMode=BMM_Additive
    translationMode=BMM_Additive
    translationSpace=BCS_BoneSpace
    alphaBoolBlend {blendOption=Linear}
    boneToModify {boneName=dyn_pieace_14}
   }
   n110 = "Transform (Modify) Bone - Bone: dyn_pieace_15"(ComponentPose=@n109, Alpha=1, Translation="0,0,0", Rotation="0,0,0", Scale="1,1,1") {
    rotationMode=BMM_Additive
    translationMode=BMM_Additive
    translationSpace=BCS_BoneSpace
    alphaBoolBlend {blendOption=Linear}
    boneToModify {boneName=dyn_pieace_15}
   }
   n111 = "Transform (Modify) Bone - Bone: dyn_pieace_16"(ComponentPose=@n110, Alpha=1, Translation="0,0,0", Rotation="0,0,0", Scale="1,1,1") {
    rotationMode=BMM_Additive
    translationMode=BMM_Additive
    translationSpace=BCS_BoneSpace
    alphaBoolBlend {blendOption=Linear}
    boneToModify {boneName=dyn_pieace_16}
   }
   n112 = "Transform (Modify) Bone - Bone: dyn_pieace_17"(ComponentPose=@n111, Alpha=1, Translation="0,0,0", Rotation="0,0,0", Scale="1,1,1") {
    rotationMode=BMM_Additive
    translationMode=BMM_Additive
    translationSpace=BCS_BoneSpace
    alphaBoolBlend {blendOption=Linear}
    boneToModify {boneName=dyn_pieace_17}
   }
   n113 = "Transform (Modify) Bone - Bone: dyn_pieace_18"(ComponentPose=@n112, Alpha=1, Translation="0,0,0", Rotation="0,0,0", Scale="1,1,1") {
    rotationMode=BMM_Additive
    translationMode=BMM_Additive
    translationSpace=BCS_BoneSpace
    alphaBoolBlend {blendOption=Linear}
    boneToModify {boneName=dyn_pieace_18}
   }
   n114 = "Transform (Modify) Bone - Bone: dyn_pieace_19"(ComponentPose=@n113, Alpha=1, Translation="0,0,0", Rotation="0,0,0", Scale="1,1,1") {
    rotationMode=BMM_Additive
    translationMode=BMM_Additive
    translationSpace=BCS_BoneSpace
    alphaBoolBlend {blendOption=Linear}
    boneToModify {boneName=dyn_pieace_19}
   }
   n115 = "Transform (Modify) Bone - Bone: dyn_pieace_20"(ComponentPose=@n114, Alpha=1, Translation="0,0,0", Rotation="0,0,0", Scale="1,1,1") {
    rotationMode=BMM_Additive
    translationMode=BMM_Additive
    translationSpace=BCS_BoneSpace
    alphaBoolBlend {blendInTime=0.5, blendOption=CircularIn, blendOutTime=0.5}
    boneToModify {boneName=dyn_pieace_20}
   }
   Save cached pose 'FrontendDynamics'(Pose=Component To Local(ComponentPose=@n115)) {cacheName=FrontendDynamics}
  ]}
 EventGraph {nodes [
   Event BlueprintUpdateAnimation()
   > n1 = Sequence!(execute) -> (then_0, then_1, then_2:Exec, then_3:Exec, then_4:Exec, then_5:Exec, then_6:Exec, then_7:Exec, then_8:Exec)
   n2 = GetOwnerSkeletalMeshComponent()
   > n3 = Random Vector Function!(
    execute
    CharacterPartAnimInstance:Custom Character Part Anim Instance Object Reference
    SkelMeshComponent:Mesh Component Object Reference=@n2
    JointName:Name=None
    GameTimeInSeconds:double=Get customSpeed() -> (customSpeed:double)
    TimeOffset:double=0
    PerlinScale:double=1
    Frequency:Vector="0.6,0.6,0.6"
    Amplitude:Vector="1.8,1.8,2"
    Offset:Vector="3,3,2"
   ) -> (then, VectorOut:Vector)
   > Set sine1(sine1:Vector=@n3) -> (Output_Get:Vector)
   n4 = float * float(A=1, B=30) {is=title K2Node_PromotableOperator}
   n5 = GetRealTimeSeconds()
   n6 = float * float(A="Sin (Radians)"(A=float * float(A=@n5, B=2) {is=title K2Node_CommutativeAssociativeBinaryOperator}), B=5)
   > Set sin_rot_01(sin_rot_01:Rotator=Rotation From X Vector!(InVec:Vector=MakeVector(X=@n4, Y=@n6, Z=@n6)) -> (ReturnValue:Rotator)) -> (Output_Get:Rotator)
   n7 = Random Vector Function(
    execute=@n1.then_1
    SkelMeshComponent=@n2
    JointName=None
    GameTimeInSeconds=Get customSpeed()
    TimeOffset=0
    PerlinScale=1
    Frequency="0.35,0.35,0.35"
    Amplitude="3,3,2"
    Offset="-5,-5,-3"
   )
   > Set sine2(sine2:Vector=@n7) -> (Output_Get:Vector)
   n8 = float * float(A=1, B=30) {is=title K2Node_PromotableOperator}
   n9 = float * float(A="Sin (Radians)"(A=float * float(A=@n5, B=1) {is=title K2Node_CommutativeAssociativeBinaryOperator}), B=5)
   > Set sin_rot_02(sin_rot_02:Rotator=Rotation From X Vector(InVec=MakeVector(X=@n8, Y=@n9, Z=@n9))) -> (Output_Get:Rotator)
   n10 = Random Vector Function(
    execute=@n1.then_2
    SkelMeshComponent=@n2
    JointName=None
    GameTimeInSeconds=Get customSpeed()
    TimeOffset=0
    PerlinScale=1
    Frequency="0.4,0.4,0.4"
    Amplitude="1,2.3,2.3"
    Offset="1,2,2"
   )
   > Set sine3(sine3:Vector=@n10) -> (Output_Get:Vector)
   n11 = float * float(A=1, B=30) {is=title K2Node_PromotableOperator}
   n12 = GetRealTimeSeconds()
   n13 = float * float(A="Sin (Radians)"(A=float * float(A=@n12, B=3) {is=title K2Node_CommutativeAssociativeBinaryOperator}), B=5)
   > Set sin_rot_03(sin_rot_03:Rotator=Rotation From X Vector(InVec=MakeVector(X=@n11, Y=@n13, Z=@n13))) -> (Output_Get:Rotator)
   n14 = Random Vector Function(
    execute=@n1.then_3
    SkelMeshComponent=@n2
    JointName=None
    GameTimeInSeconds=Get customSpeed()
    TimeOffset=0
    PerlinScale=1
    Frequency="0.7,0.7,0.7"
    Amplitude="2.4,2.4,2.4"
    Offset="4,4,4"
   )
   > Set sine4(sine4:Vector=@n14) -> (Output_Get:Vector)
   n15 = float * float(A=1, B=30) {is=title K2Node_PromotableOperator}
   n16 = GetRealTimeSeconds()
   n17 = float * float(A="Sin (Radians)"(A=float * float(A=@n16, B=3) {is=title K2Node_CommutativeAssociativeBinaryOperator}), B=5)
   > Set sin_rot_04(sin_rot_04:Rotator=Rotation From X Vector(InVec=MakeVector(X=@n15, Y=@n17, Z=@n17))) -> (Output_Get:Rotator)
   n18 = Random Vector Function(
    execute=@n1.then_4
    SkelMeshComponent=@n2
    JointName=None
    GameTimeInSeconds=Get customSpeed()
    TimeOffset=0
    PerlinScale=1
    Frequency="0.4,0.4,0.4"
    Amplitude="3.5,3.5,3.5"
    Offset="3,3,3"
   )
   > Set sine5(sine5:Vector=@n18) -> (Output_Get:Vector)
   n19 = Random Vector Function(
    execute=@n1.then_5
    SkelMeshComponent=@n2
    JointName=None
    GameTimeInSeconds=Get customSpeed()
    TimeOffset=0
    PerlinScale=1
    Frequency="0.7,0.7,0.7"
    Amplitude="3.5,3.5,3.5"
    Offset="2,2,2"
   )
   > Set sine6(sine6:Vector=@n19) -> (Output_Get:Vector)
   n20 = Random Vector Function(
    execute=@n1.then_6
    SkelMeshComponent=@n2
    JointName=None
    GameTimeInSeconds=Get customSpeed()
    TimeOffset=0
    PerlinScale=1
    Frequency="0.5,0.5,0.5"
    Amplitude="2,2,2"
    Offset="0,0,0"
   )
   > Set sine7(sine7:Vector=@n20) -> (Output_Get:Vector)
   Set customSpeed(execute=@n1.then_7, customSpeed:double=float / float(A=GetRealTimeSeconds(), B=1.25) {is=title K2Node_CallFunction}) -> (Output_Get:double)
   n21 = Get bIsInFrontEnd()
   Set frontendAlpha(execute=@n1.then_8, frontendAlpha:double=SelectFloat(A=0.51, B=0, bPickA=@n21)) -> (Output_Get:double)
   > n22 = Branch(Condition=AND Boolean(A=@n21, B=Get bIsOnPawn() -> (bIsOnPawn:bool)))
   > Set FE_Translation_Pieces(FE_Translation_Pieces:Byte=4) -> (Output_Get:Byte=0)
   Set FE_Translation_Pieces(execute=@n22.else, FE_Translation_Pieces=0) -> (Output_Get=0)
   TryGetPawnOwner()
   GetRealTimeSeconds()
  ]}
]
variables [
 customSpeed {type=number}
 FE_Translation_Pieces {type=integer}
 FE_withpawn {type=boolean}
 frontendAlpha {type=number}
 sin_rot_01 {type=Rotator}
 sin_rot_02 {type=Rotator}
 sin_rot_03 {type=Rotator}
 sin_rot_04 {type=Rotator}
 sine1 {type=Vector}
 sine2 {type=Vector}
 sine3 {type=Vector}
 sine4 {type=Vector}
 sine5 {type=Vector}
 sine6 {type=Vector}
 sine7 {type=Vector}
]
